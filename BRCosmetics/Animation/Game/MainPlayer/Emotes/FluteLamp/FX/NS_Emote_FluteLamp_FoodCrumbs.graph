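# Niagara system for the FluteLamp emote that throws out a short burst of food crumbs, goop and sparkles.
# Mouth_goop_spray_v2 sprays stretched goop meshes from a socket in a brief cone; its v01 predecessor is disabled.
# Bread crumbs, red/yellow/white goop and rising curl-noise sparkles burst around the system origin.
# In goopy002, spawn groups drive visibility tags and per-group motion, mixing speed-scaled goop with tumbling food bits from five mesh variants.
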NiagaraSystem /BRCosmetics/Animation/Game/MainPlayer/Emotes/FluteLamp/FX/NS_Emote_FluteLamp_FoodCrumbs
dependencies {
 usedDataInterfaces [
  &NiagaraDataInterfaceCurve
  &NiagaraDataInterfaceArrayColor
  &NiagaraDataInterfaceArrayInt32
  &NiagaraDataInterfaceArrayFloat3
  &NiagaraDataInterfaceSocketReader
 ]
 usedDynamicInputs [/Niagara/DynamicInputs/TypeConversions/Vector2DLength, &RandomRangeVector2D]
 usedModules [
  &self:SystemSpawnScript.SystemScriptSource.SystemScriptGraph.NiagaraNodeAssignment_0.SetVariables_#1
  &SystemState
 ]
}
emitters {
 Mouth_goop_spray_v01 {
  bEnabled=false
  emitterUpdateScript [
   &EmitterState(Inactive Response=::NewEnumerator1, Life Cycle Mode=::NewEnumerator1, Loop Behavior=::NewEnumerator1, Loop Duration=1, UseLoopDelay=false)
   &SpawnRate(
    SpawnGroup=1
    SpawnRate=&FloatFromCurve(
     FloatCurve=DataInterface() {propertyValues {Curve {keys [<at 0 1>,<at 0.24 1>,<at 0.35 0>]}}}
     CurveIndex=$Emitter.NormalizedLoopAge
     Scale Curve=250
    )
    Use Spawn Probability=false
   ) {module=SpawnRate001}
   &SpawnBurst_Instantaneous(Spawn Count=15, Spawn Group=1, Use Spawn Probability=false, UseLoopCountLimit=false) {module=SpawnBurst_Instantaneous001}
   &SpawnBurst_Instantaneous(Spawn Count=11, Use Spawn Probability=false, UseLoopCountLimit=false)
   &SpawnRate(
    SpawnRate=&FloatFromCurve(
     FloatCurve=DataInterface() {propertyValues {Curve {keys [<at 0 1>,<at 0.2473 1>,<at 0.35 0>]}}}
     CurveIndex=$Emitter.NormalizedLoopAge
     Scale Curve=350
    )
    Use Spawn Probability=false
   )
   /SetVariables(
    Emitter.Ketchup 1:LinearColor=<rgba 0.1448 0.0058 0.0024 1>
    Emitter.Ketchup 2:LinearColor=<rgba 0.4906 0 0 1>
    Emitter.Mustard 2:LinearColor=<rgba 1 0.59 0 1>
    Emitter.mustard 1:LinearColor=<rgba 0.5323 0.3141 0 1>
   ) {isSetParameters=true, module=SetVariables_#2}
  ]
  particleSpawnScript [
   /SetVariables(Particles.VisibilityTag=&SpawnGroup()) {isSetParameters=true, module=SetVariables_#3}
   &InitializeParticle(
    Color Mode=::NewEnumerator0
    Lifetime Max=&RandomRangeFloat(Minimum=0.25, Maximum=0.35)
    Lifetime Min:NiagaraFloat=&RandomRangeFloat(Minimum=0.18, Maximum=0.2)
    Lifetime Mode=::NewEnumerator1
    Mesh Scale Max=<0.7 1.7 5>
    Mesh Scale Min=<0.4 0.4 2>
    Mesh Scale Mode=::NewEnumerator2
    Sprite Size Mode=::NewEnumerator1
    Uniform Sprite Size Max=5
    Uniform Sprite Size Min=2
    UsePositionOffset=false
    Write Material Random=true
   )
   &InitialMeshOrientation(Mesh Orientation Mode=::NewEnumerator7, Rotation=<0 -0.25 0>, Use Rotation Vector=true)
   &DynamicMaterialParameters(
    Index 0 Param 1=1
    Index 0 Param 2=1
    Module.Param0WriteEnabled=true
    Module.Param1WriteEnabled=true
    Module.Param2WriteEnabled=false
    Module.Param3WriteEnabled=false
   )
   &AddVelocityFromPoint(
    Module.Use Falloff Distance=false
    Origin Offset=<-1 0 0.5>
    Velocity Strength=&RandomRangeFloat(Minimum=-80, Maximum=-350)
   ) {enabled=false}
   &ShapeLocation(
    Cylinder Height=5
    Cylinder Radius=0.3
    MaskSpawnGroup=false
    Offset=<0 0 1>
    Offset Mode=::NewEnumerator0
    Override Seed=false
    Shape Primitive=::NewEnumerator1
    Surface Only=false
    UseOwnerScale=true
   )
   &AddVelocity(
    Cone Angle=100
    Cone Axis=<-2 0 0.2>
    Module.UseConeFalloff=false
    Module.UseVelDistribution=true
    Override Seed=false
    Rotation Angle=20
    Rotation Axis:Vector3f=<0 1 0>
    Rotation Mode=::NewEnumerator1
    Velocity Mode=::NewEnumerator2
    Velocity Speed=&MergeHelper_FloatSwitch(
     0=&RandomRangeFloat(Minimum=200, Maximum=500)
     1=&RandomRangeFloat(Minimum=200, Maximum=510)
     2=&RandomRangeFloat(Minimum=150, Maximum=250)
     3=?
     Index=$Particles.VisibilityTag
     Num=::NewEnumerator0
    )
   )
   &CameraOffset(Camera Offset Amount=&RandomRangeFloat(Minimum=0, Maximum=3))
   &SocketLocation(
    Orientation Sampling=::NewEnumerator1
    SamplePosOffset=true
    Sampled Position Offset=<-4.7 6.3 0>
    Socket Reader=$System.Socket Reader
    Socket Sampling Mode=::NewEnumerator2
    VelocitySpeedLimit=false
   )
   &AddVelocity(
    Rotation Mode=::NewEnumerator5
    Velocity=&RandomRangeVector(Minimum=<-500 -100 -50>, Maximum=<-150 100 50>)
   ) {enabled=false, module=AddVelocity001}
  ]
  particleUpdateScript [
   &ParticleState()
   &GravityForce(Gravity=&RandomRangeVector(Minimum=<0 0 -400>, Maximum=<0 0 -980>))
   &Drag(Drag=&RandomRangeFloat(Minimum=2, Maximum=5), Use Linear Drag=true, Use Rotational Drag=false)
   &SolveForcesAndVelocity(Clamp Velocity=false, Limit Acceleration=false)
   &ScaleMeshSize(Scale Factor=&VectorFromFloat(Value=&FloatFromCurve(
      FloatCurve=DataInterface() {propertyValues {Curve {keys [
          <at 0 0 RCIM_Cubic>
          {interpMode=RCIM_Cubic, tangentMode=RCTM_User, time=0.2169, value=1}
          {interpMode=RCIM_Cubic, tangentMode=RCTM_User, time=0.5, value=1}
          {arriveTangent=-1.4032, interpMode=RCIM_Cubic, leaveTangent=-1.4032, tangentMode=RCTM_User, time=0.6, value=0}
         ]}}}
      Scale Curve=0.55
     )))
   &ScaleMeshSizeBySpeed(
    Max Scale Factor=&RandomRangeVector(Minimum=<0.7 0.7 0.8>, Maximum=<1.3 1.3 1.5>)
    Min Scale Factor:Vector3f=&RandomRangeVector(Minimum=<0.2 0.2 0.35>, Maximum=<0.45 0.45 0.6>)
    Sample Scale Factor By Curve=true
    Scale Factor Curve=DataInterface() {propertyValues {Curve {keys [
        {interpMode=RCIM_Cubic, leaveTangent=-0.1696, tangentMode=RCTM_Break, time=0, value=0.4}
        {interpMode=RCIM_Cubic, leaveTangent=-0.1696, tangentMode=RCTM_Break, time=0.4, value=1}
        {interpMode=RCIM_Cubic, leaveTangent=-0.1696, tangentMode=RCTM_Break, time=0.8, value=0}
       ]}}}
    Speed Threshold=285
   )
   &ScaleColor(
    Scale Alpha=&FloatFromCurve(FloatCurve=DataInterface() {propertyValues {Curve {keys [
         {tangentMode=RCTM_User, time=0, value=0}
         <at 0.0599 1>
         {arriveTangent=-0.0405, interpMode=RCIM_Cubic, leaveTangent=-0.0405, tangentMode=RCTM_User, time=0.7009, value=1}
         {arriveTangent=-8.0089, interpMode=RCIM_Cubic, leaveTangent=-8.0089, tangentMode=RCTM_User, time=1, value=0}
        ]}}})
    ScaleA=true
    ScaleRGB=true
   )
  ]
  rendererClasses [&NiagaraMeshRendererProperties]
  renderers [
   {rendererClass=&NiagaraMeshRendererProperties, rendererIndex=0}
   {rendererClass=&NiagaraMeshRendererProperties, rendererIndex=1}
  ]
 }
 Mouth_goop_spray_v2 {
  emitterUpdateScript [
   &EmitterState(Inactive Response=::NewEnumerator1, Life Cycle Mode=::NewEnumerator1, Loop Behavior=::NewEnumerator1, Loop Duration=0.5, UseLoopDelay=false)
   &SpawnRate(
    SpawnGroup=1
    SpawnRate=&FloatFromCurve(
     FloatCurve=DataInterface() {propertyValues {Curve {keys [<at 0 1>,<at 0.15 1>,<at 0.16 0>]}}}
     CurveIndex=$Emitter.NormalizedLoopAge
     Scale Curve=250
    )
    Use Spawn Probability=false
   ) {module=SpawnRate001}
   &SpawnRate(
    SpawnRate=&FloatFromCurve(
     FloatCurve=DataInterface() {propertyValues {Curve {keys [<at 0 1>,<at 0.15 1>,<at 0.16 0>]}}}
     CurveIndex=$Emitter.NormalizedLoopAge
     Scale Curve=350
    )
    Use Spawn Probability=false
   )
   /SetVariables(
    Emitter.Ketchup 1:LinearColor=<rgba 0.1448 0.0058 0.0024 1>
    Emitter.Ketchup 2:LinearColor=<rgba 0.4906 0 0 1>
    Emitter.Mustard 2:LinearColor=<rgba 1 0.59 0 1>
    Emitter.mustard 1:LinearColor=<rgba 0.5323 0.3141 0 1>
   ) {isSetParameters=true, module=SetVariables_#2}
  ]
  particleSpawnScript [
   /SetVariables(Particles.VisibilityTag=&SpawnGroup()) {isSetParameters=true, module=SetVariables_#3}
   &InitializeParticle(
    Color Mode=::NewEnumerator0
    Lifetime Max=&RandomRangeFloat(Minimum=0.25, Maximum=0.4)
    Lifetime Min:NiagaraFloat=&RandomRangeFloat(Minimum=0.15, Maximum=0.2)
    Lifetime Mode=::NewEnumerator1
    Mesh Scale Max=<0.7 1.7 5>
    Mesh Scale Min=<0.4 0.4 2>
    Mesh Scale Mode=::NewEnumerator2
    Sprite Size Mode=::NewEnumerator1
    Uniform Sprite Size Max=5
    Uniform Sprite Size Min=2
    UsePositionOffset=false
    Write Material Random=true
   )
   &InitialMeshOrientation(Mesh Orientation Mode=::NewEnumerator7, Rotation=<0 -0.25 0>, Use Rotation Vector=true)
   &DynamicMaterialParameters(
    Index 0 Param 1=1
    Index 0 Param 2=1
    Module.Param0WriteEnabled=true
    Module.Param1WriteEnabled=true
    Module.Param2WriteEnabled=false
    Module.Param3WriteEnabled=false
   )
   &ShapeLocation(
    Cylinder Height=5
    Cylinder Radius=0.3
    MaskSpawnGroup=false
    Offset=<0 0 1>
    Offset Mode=::NewEnumerator0
    Override Seed=false
    Shape Primitive=::NewEnumerator1
    Surface Only=false
    UseOwnerScale=true
   )
   &AddVelocity(
    Cone Angle=90
    Cone Axis=<-2 0 0.2>
    Module.UseConeFalloff=false
    Module.UseVelDistribution=true
    Override Seed=false
    Rotation Angle=20
    Rotation Axis:Vector3f=<0 1 0>
    Rotation Mode=::NewEnumerator1
    Velocity Mode=::NewEnumerator2
    Velocity Speed=&MergeHelper_FloatSwitch(
     0=&RandomRangeFloat(
      Minimum=200
      Maximum=&FloatFromCurve(FloatCurve=DataInterface() {propertyValues {Curve {keys [<at 0.0043 480>,<at 0.1505 0>]}}})
     )
     1=&RandomRangeFloat(
      Minimum=200
      Maximum=&FloatFromCurve(FloatCurve=DataInterface() {propertyValues {Curve {keys [<at 0 500>,<at 1 0>]}}})
     )
     2=&RandomRangeFloat(Minimum=150, Maximum=250)
     3=?
     Index=$Particles.VisibilityTag
     Num=::NewEnumerator0
    )
   )
   &CameraOffset(Camera Offset Amount=&RandomRangeFloat(Minimum=0, Maximum=3))
   &SocketLocation(
    Orientation Sampling=::NewEnumerator1
    SamplePosOffset=true
    Sampled Position Offset=<-4.7 6.3 0>
    Socket Reader=$System.Socket Reader
    Socket Sampling Mode=::NewEnumerator2
    VelocitySpeedLimit=false
   )
  ]
  particleUpdateScript [
   &ParticleState()
   &GravityForce(Gravity=&RandomRangeVector(Minimum=<0 0 -600>, Maximum=<0 0 -1500>))
   &Drag(Drag=&RandomRangeFloat(Minimum=2, Maximum=5), Use Linear Drag=true, Use Rotational Drag=false)
   &SolveForcesAndVelocity(Clamp Velocity=false, Limit Acceleration=false)
   &ScaleMeshSize(Scale Factor=&VectorFromFloat(Value=&FloatFromCurve(
      FloatCurve=DataInterface() {propertyValues {Curve {keys [
          <at 0 0 RCIM_Cubic>
          {interpMode=RCIM_Cubic, tangentMode=RCTM_User, time=0.02, value=1}
          {interpMode=RCIM_Cubic, tangentMode=RCTM_User, time=0.3232, value=1}
          {arriveTangent=-1.4032, interpMode=RCIM_Cubic, leaveTangent=-1.4032, tangentMode=RCTM_User, time=0.57, value=0}
         ]}}}
      Scale Curve=0.55
     )))
   &ScaleMeshSizeBySpeed(
    Max Scale Factor=&RandomRangeVector(Minimum=<0.8 0.8 0.9>, Maximum=<1.4 1.4 1.6>)
    Min Scale Factor:Vector3f=&RandomRangeVector(Minimum=<0.3 0.3 0.4>, Maximum=<0.6 0.6 0.7>)
    Sample Scale Factor By Curve=true
    Scale Factor Curve=DataInterface() {propertyValues {Curve {keys [
        {interpMode=RCIM_Cubic, leaveTangent=-0.1696, tangentMode=RCTM_Break, time=0, value=0.4}
        {interpMode=RCIM_Cubic, leaveTangent=-0.1696, tangentMode=RCTM_Break, time=0.4, value=1}
        {interpMode=RCIM_Cubic, leaveTangent=-0.1696, tangentMode=RCTM_Break, time=0.8, value=0}
       ]}}}
    Speed Threshold=290
   )
   &ScaleColor(
    Scale Alpha=&FloatFromCurve(FloatCurve=DataInterface() {propertyValues {Curve {keys [
         {tangentMode=RCTM_User, time=0, value=0}
         <at 0.0599 1>
         {arriveTangent=-0.0405, interpMode=RCIM_Cubic, leaveTangent=-0.0405, tangentMode=RCTM_User, time=0.7009, value=1}
         {arriveTangent=-8.0089, interpMode=RCIM_Cubic, leaveTangent=-8.0089, tangentMode=RCTM_User, time=1, value=0}
        ]}}})
    ScaleA=true
    ScaleRGB=true
   )
  ]
  rendererClasses [&NiagaraMeshRendererProperties]
  renderers [
   {rendererClass=&NiagaraMeshRendererProperties, rendererIndex=0}
   {rendererClass=&NiagaraMeshRendererProperties, rendererIndex=1}
  ]
 }
 Sparkles001 {
  emitterUpdateScript [
   &EmitterState()
   &SpawnBurst_Instantaneous(Spawn Count=13, Use Spawn Probability=false)
  ]
  particleSpawnScript [
   &InitializeParticle(
    Color=<rgba 5 3.7043 1.51 1>
    Lifetime=&UniformRangedFloat(Minimum=0.3, Maximum=0.45)
    Position Offset=<-17 0 5>
    Sprite Size Mode=::NewEnumerator3
    Uniform Sprite Size=&UniformRangedFloat(Minimum=3, Maximum=7)
    UsePositionOffset=true
    Write Material Random=true
   ) {module=InitializeParticle001}
   &ShapeLocation(MaskSpawnGroup=false, Override Seed=false, Sphere Radius=5, UseOwnerScale=true)
   &AddVelocity(Velocity=&UniformRangedVector(Minimum=<-150 -150 -80>, Maximum=<150 150 50>))
   &ApplyInitialForces()
  ]
  particleUpdateScript [
   &ParticleState()
   &GravityForce(Gravity=<0 0 170>)
   /Niagara/Modules/Update/Forces/CurlNoiseForce(Noise Frequency=0.02, Noise Quality / Cost=::NewEnumerator0, Noise Strength=150, Pan Noise=false)
   &Drag(Drag=&RandomRangeFloat(Minimum=0, Maximum=5), Use Linear Drag=true, Use Rotational Drag=false)
   &SolveForcesAndVelocity(Clamp Velocity=false, Limit Acceleration=false)
   &ScaleColor(
    Scale Alpha=&FloatFromCurve(FloatCurve=DataInterface() {propertyValues {Curve {keys [<at -0.0092 0.0144>,<at 0.0967 0.9767>,<at 1.0011 0.002>]}}})
    ScaleA=true
    ScaleRGB=true
   )
   &ScaleSpriteSize(Scale Factor=&Vector2DFromFloat(Value=&FloatFromCurve(FloatCurve=DataInterface() {propertyValues {Curve {keys [<at -0.0046 0>,<at 0.1746 0.9952>,<at 1.0011 0.002>]}}})))
  ]
  rendererClasses [&NiagaraSpriteRendererProperties]
  renderers [{rendererClass=&NiagaraSpriteRendererProperties}]
 }
 bread_crrumbs {
  emitterUpdateScript [
   &EmitterState(Life Cycle Mode=::NewEnumerator1, Loop Behavior=::NewEnumerator1, Loop Delay=0.1, Loop Duration=1, UseLoopDelay=true)
   &SpawnBurst_Instantaneous(Spawn Count=&RandomRangeInt(Minimum=4, Maximum=7), Spawn Time=0.1, Use Spawn Probability=false)
   &SpawnRate(
    SpawnRate=&RandomRangeFloat(Minimum=10, Maximum=40, Evaluation Type=::NewEnumerator1)
    Use Spawn Probability=false
   ) {enabled=false}
  ]
  particleSpawnScript [
   /Niagara/Modules/Spawn/Initialization/InitializeParticle(
    Color=<rgba 2 2 2 1>
    Mesh Scale=&Multiply_Vector(
     A=<1 1 1>
     B=&VectorFromFloat(Value=&UniformRangedFloat(Minimum=0.06, Maximum=0.1))
    )
    Module.Write Color=true
    Module.Write Lifetime=true
    Module.Write Mass=true
    Module.Write Material Random=true
    Module.Write Position=false
    Module.Write Scale=true
    Module.Write SpriteRotation=false
    Module.Write SpriteSize=false
    Position=&SimulationPosition()
   ) {enabled=false}
   /Niagara/Modules/Spawn/Initialization/V2/InitializeParticle(
    Color=<rgba 2 2 2 1>
    Lifetime=&RandomRangeFloat(Minimum=0.4, Maximum=0.75)
    Mass Mode=::NewEnumerator1
    Mesh Scale=&Multiply_Vector(
     A=<1 1 1>
     B=&VectorFromFloat(Value=&UniformRangedFloat(Minimum=0.06, Maximum=0.1))
    )
    Mesh Scale Mode=::NewEnumerator4
    Sprite Size Mode=::NewEnumerator4
    UsePositionOffset=false
    Write Material Random=true
   ) {module=InitializeParticle001}
   &SocketLocation(
    SamplePosOffset=true
    Sampled Position Offset=<-10 10 0>
    Socket Reader=$System.Socket Reader
    Socket Sampling Mode=::NewEnumerator3
    VelocitySpeedLimit=false
   ) {enabled=false}
   &InitialMeshOrientation(Use Rotation Vector=false) {module=InitialMeshOrientation001}
   /SetVariables(Particles.RotationOrientation:NiagaraFloat=&UniformRangedFloat(Minimum=-1, Maximum=1)) {enabled=false, isSetParameters=true}
   /SetVariables(Particles.RotationSpeed:Vector3f=&UniformRangedVector(Minimum=<-1 -1 -1>, Maximum=<1 1 1>)) {enabled=false, isSetParameters=true}
   &ShapeLocation(MaskSpawnGroup=false, Override Seed=false, Sphere Radius=5, UseOwnerScale=true)
   &AddVelocityFromPoint(Module.Use Falloff Distance=false, Velocity Strength=&RandomRangeFloat(Minimum=150, Maximum=-180))
  ]
  particleUpdateScript [
   &ParticleState()
   &GravityForce(Gravity=&RandomRangeVector(Minimum=<0 0 -200>, Maximum=<0 0 -400>))
   &InheritSourceMovement(
    Applied Position Scale=&VectorFromFloat(Value=&FloatFromCurve(FloatCurve=DataInterface() {propertyValues {Curve {keys [{arriveTangent=-3.1099, interpMode=RCIM_Cubic, leaveTangent=-3.1099, tangentMode=RCTM_User, time=0, value=1}, <at 1 0>]}}}))
    ApplyForce=false
    ApplyPosition=true
    ApplyPositionFalloff=false
   ) {enabled=false}
   &AccelerationForce(Acceleration=<50 50 0>)
   &SolveForcesAndVelocity(Clamp Velocity=false, Limit Acceleration=false)
   &ScaleMeshSize(Scale Factor=&Multiply_Vector(
     A=<1 1 1>
     B=&VectorFromFloat(Value=&FloatFromCurve(FloatCurve=DataInterface() {propertyValues {Curve {keys [<at -0.0046 0>,<at 0.0542 0.0143>,<at 0.0756 0.5>,<at 1.0011 0>]}}}))
    ))
  ]
  rendererClasses [&NiagaraMeshRendererProperties]
  renderers [{rendererClass=&NiagaraMeshRendererProperties, rendererIndex=0}]
 }
 goopy {
  emitterUpdateScript [
   &EmitterState(
    Inactive Response=::NewEnumerator1
    Life Cycle Mode=::NewEnumerator1
    Loop Behavior=::NewEnumerator1
    Loop Delay=0.1
    Loop Duration=1
    UseLoopDelay=true
   )
   &SpawnBurst_Instantaneous(Spawn Count=6, Spawn Group=1, Spawn Time=0.1, Use Spawn Probability=false, UseLoopCountLimit=false) {module=SpawnBurst_Instantaneous001}
   &SpawnBurst_Instantaneous(Spawn Count=3, Spawn Time=0.1, Use Spawn Probability=false, UseLoopCountLimit=false)
   &SpawnRate(Spawn Probability=0.1, SpawnRate=50, Use Spawn Probability=true) {enabled=false}
   /SetVariables(
    Emitter.Ketchup 1:LinearColor=<rgba 0.1448 0.0058 0.0024 1>
    Emitter.Ketchup 2:LinearColor=<rgba 0.4906 0 0 1>
    Emitter.Mustard 2:LinearColor=<rgba 1 0.59 0 1>
    Emitter.mustard 1:LinearColor=<rgba 0.5323 0.3141 0 1>
   ) {isSetParameters=true, module=SetVariables_#2}
  ]
  particleSpawnScript [
   /SetVariables(Particles.VisibilityTag=&SpawnGroup()) {isSetParameters=true, module=SetVariables_#3}
   &InitializeParticle(
    Color=&SelectLinearColorFromArray(
     Array Sampling Mode=::NewEnumerator0
     Direct Array Index=?
     Color Selection Array=DataInterface() {propertyValues {ColorData [<rgba 1 1 1 0.5>, <rgba 1 0.0401 0.0167 0.5>, <rgba 1 0.59 0 0.5>]}}
     Random Seed=0
    )
    Lifetime Max=0.4
    Lifetime Min=0.2
    Lifetime Mode=::NewEnumerator1
    Mesh Scale Max=<0.5 2 4>
    Mesh Scale Min=<0.2 0.25 1>
    Mesh Scale Mode=::NewEnumerator2
    Sprite Size Mode=::NewEnumerator1
    Uniform Sprite Size Max=5
    Uniform Sprite Size Min=2
    UsePositionOffset=false
    Write Material Random=true
   )
   &InitialMeshOrientation(Mesh Orientation Mode=::NewEnumerator7, Rotation=<0 -0.25 0>, Use Rotation Vector=true)
   &DynamicMaterialParameters(
    Index 0 Param 1=1
    Index 0 Param 2=1
    Module.Param0WriteEnabled=true
    Module.Param1WriteEnabled=true
    Module.Param2WriteEnabled=false
    Module.Param3WriteEnabled=false
   )
   &AddVelocityFromPoint(Module.Use Falloff Distance=false, Velocity Strength=&RandomRangeFloat(Minimum=-150, Maximum=-180))
   &ShapeLocation(MaskSpawnGroup=false, Override Seed=false, Sphere Radius=5, UseOwnerScale=true)
  ]
  particleUpdateScript [
   &ParticleState()
   &GravityForce(Gravity=&RandomRangeVector(Minimum=<0 0 -200>, Maximum=<0 0 -400>))
   &Drag(Drag=3, Use Linear Drag=true, Use Rotational Drag=false)
   &SolveForcesAndVelocity(Clamp Velocity=false, Limit Acceleration=false)
   &ScaleMeshSize(Scale Factor=&VectorFromFloat(Value=&FloatFromCurve(
      FloatCurve=DataInterface() {propertyValues {Curve {keys [
          <at 0 0 RCIM_Cubic>
          {interpMode=RCIM_Cubic, tangentMode=RCTM_User, time=0.2169, value=1}
          {interpMode=RCIM_Cubic, tangentMode=RCTM_User, time=0.5, value=1}
          {arriveTangent=-1.4032, interpMode=RCIM_Cubic, leaveTangent=-1.4032, tangentMode=RCTM_User, time=1, value=0.5}
         ]}}}
      Scale Curve=0.5
     )))
   &ScaleMeshSizeBySpeed(
    Max Scale Factor=<1 1 2>
    Min Scale Factor:Vector3f=<1 1 0.33>
    Sample Scale Factor By Curve=true
    Scale Factor Curve=DataInterface() {propertyValues {Curve {defaultValue=0, keys [{tangentMode=RCTM_User, time=0, value=1}, {tangentMode=RCTM_User, time=0.8, value=0}]}}}
    Speed Threshold=200
   )
   &ScaleColor(
    Scale Alpha=&FloatFromCurve(FloatCurve=DataInterface() {propertyValues {Curve {keys [
         {tangentMode=RCTM_User, time=0, value=0}
         <at 0.0599 1>
         {arriveTangent=-0.0405, interpMode=RCIM_Cubic, leaveTangent=-0.0405, tangentMode=RCTM_User, time=0.7009, value=1}
         {arriveTangent=-8.0089, interpMode=RCIM_Cubic, leaveTangent=-8.0089, tangentMode=RCTM_User, time=1, value=0}
        ]}}})
    ScaleA=true
    ScaleRGB=true
   )
  ]
  rendererClasses [&NiagaraMeshRendererProperties]
  renderers [
   {rendererClass=&NiagaraMeshRendererProperties, rendererIndex=0}
   {rendererClass=&NiagaraMeshRendererProperties, rendererIndex=1}
  ]
 }
 goopy002 {
  emitterUpdateScript [
   &EmitterState(Inactive Response=::NewEnumerator1, Life Cycle Mode=::NewEnumerator1, Loop Behavior=::NewEnumerator1, Loop Duration=1, UseLoopDelay=false)
   &SpawnBurst_Instantaneous(Spawn Count=2, Spawn Group=1, Spawn Time=0.05, Use Spawn Probability=false, UseLoopCountLimit=false) {module=SpawnBurst_Instantaneous003}
   &SpawnBurst_Instantaneous(Spawn Count=7, Spawn Group=2, Spawn Time=0.05, Use Spawn Probability=false, UseLoopCountLimit=false) {module=SpawnBurst_Instantaneous002}
   &SpawnBurst_Instantaneous(Spawn Count=5, Spawn Group=1, Spawn Time=0.03, Use Spawn Probability=false, UseLoopCountLimit=false) {module=SpawnBurst_Instantaneous001}
   &SpawnBurst_Instantaneous(Spawn Count=10, Use Spawn Probability=false, UseLoopCountLimit=false)
   &SpawnRate(
    Spawn Probability=0.1
    SpawnGroup=2
    SpawnRate=&FloatFromCurve(FloatCurve=DataInterface() {propertyValues {Curve {keys [<at 0 1>,<at 0.4 0>]}}}, CurveIndex=$Emitter.NormalizedLoopAge, Scale Curve=10)
    Use Spawn Probability=false
   )
   /SetVariables(
    Emitter.Ketchup 1:LinearColor=<rgba 0.1448 0.0058 0.0024 1>
    Emitter.Ketchup 2:LinearColor=<rgba 0.4906 0 0 1>
    Emitter.Mustard 2:LinearColor=<rgba 1 0.59 0 1>
    Emitter.mustard 1:LinearColor=<rgba 0.5323 0.3141 0 1>
   ) {isSetParameters=true, module=SetVariables_#2}
  ]
  particleSpawnScript [
   /SetVariables(Particles.VisibilityTag=&SpawnGroup()) {isSetParameters=true, module=SetVariables_#3}
   &InitializeParticle(
    Color=&SelectLinearColorFromArray(
     Array Sampling Mode=::NewEnumerator0
     Direct Array Index=?
     Color Selection Array=DataInterface() {propertyValues {ColorData [<rgba 1 1 1 0.5>, <rgba 1 0.0401 0.0167 0.5>, <rgba 1 0.59 0 0.5>]}}
     Random Seed=0
    )
    Lifetime=&MergeHelper_FloatSwitch(
     0=&RandomRangeFloat(Minimum=0.15, Maximum=0.25)
     1=&RandomRangeFloat(Minimum=0.17, Maximum=0.28)
     2=&RandomRangeFloat(Minimum=0.4, Maximum=0.6)
     3=?
     Index=&SpawnGroup()
     Num=::NewEnumerator1
    )
    Mesh Scale=&MergeHelper_VectorSwitch(
     0=&RandomRangeVector(Minimum=<0.2 0.25 1>, Maximum=<0.5 0.2 4>)
     1=&RandomRangeVector(Minimum=<0.2 0.25 1>, Maximum=<0.5 0.2 4>)
     2=&RandomRangeVector(Minimum=<0.15 0.15 0.15>, Maximum=<0.2 0.2 0.2>)
     3=?
     Index=&SpawnGroup()
     Num=::NewEnumerator1
    )
    Mesh Scale Mode=::NewEnumerator4
    Sprite Size Mode=::NewEnumerator1
    Uniform Sprite Size Max=5
    Uniform Sprite Size Min=2
    UsePositionOffset=false
    Write Material Random=true
   )
   &PartitionParticles(Number Of Partitions=3, Write Partition to Ribbon ID=::NewEnumerator0)
   &InitialMeshOrientation(Mesh Orientation Mode=::NewEnumerator7, Rotation=<0 -0.25 0>, Use Rotation Vector=true)
   &DynamicMaterialParameters(
    Index 0 Param 1=1
    Index 0 Param 2=1
    Module.Param0WriteEnabled=true
    Module.Param1WriteEnabled=true
    Module.Param2WriteEnabled=false
    Module.Param3WriteEnabled=false
   )
   &ShapeLocation(MaskSpawnGroup=false, Override Seed=false, Sphere Radius=5, UseOwnerScale=true)
   &AddVelocity(
    Cone Angle=80
    Cone Axis=<-1 0 0.5>
    Module.UseConeFalloff=false
    Module.UseVelDistribution=false
    Override Seed=false
    Velocity Mode=::NewEnumerator2
    Velocity Speed=&MergeHelper_FloatSwitch(
     0=&RandomRangeFloat(Minimum=50, Maximum=250)
     1=&RandomRangeFloat(Minimum=30, Maximum=250)
     2=&RandomRangeFloat(Minimum=150, Maximum=250)
     3=?
     Index=$Particles.VisibilityTag
     Num=::NewEnumerator1
    )
   )
   /SetVariables(Particles.food MeshIndex:NiagaraInt32=&SelectIntFromArray(
     Array Sampling Mode=::NewEnumerator1
     Direct Array Index=?
     Int Selection Array=DataInterface() {propertyValues {IntData [0,1,2,3,4]}}
     Random Seed=0
    )) {isSetParameters=true}
   &OffsetPosition(Position Offset=&SelectVectorFromArray(
     Direct Array Index=$Particles.VisibilityTag
     Vector Selection Array=DataInterface() {propertyValues {FloatData [<-25 0 0>, <-18 0 0>, <0 0 0>]}}
    ))
  ]
  particleUpdateScript [
   &ParticleState()
   &UpdateMeshOrientation(Rotation Vector=&RandomVector(Vector Scale=&MergeHelper_FloatSwitch(0=0, 1=0, 2=1, 3=?, Index=&SpawnGroup(), Num=::NewEnumerator1)))
   &GravityForce(Gravity=&MergeHelper_VectorSwitch(
     0=&RandomRangeVector(Minimum=<0 0 -200>, Maximum=<0 0 -400>)
     1=&RandomRangeVector(Minimum=<0 0 -200>, Maximum=<1 1 -400>)
     2=&RandomRangeVector(Minimum=<0 0 -350>, Maximum=<1 1 -600>)
     3=?
     Index=&SpawnGroup()
     Num=::NewEnumerator1
    ))
   &Drag(
    Drag=&MergeHelper_FloatSwitch(
     0=&RandomRangeFloat(Minimum=1, Maximum=7)
     1=&RandomRangeFloat(Minimum=1, Maximum=5)
     2=&RandomRangeFloat(Minimum=1, Maximum=5)
     3=?
     Index=$Particles.VisibilityTag
     Num=::NewEnumerator1
    )
    Use Linear Drag=true
    Use Rotational Drag=false
   )
   &MeshRotationForce(Rotation=&MergeHelper_VectorSwitch(0=<0 0 0>, 1=<0 0 0>, 2=<5 0 5>, 3=?, Index=&SpawnGroup(), Num=::NewEnumerator1))
   &Collision(
    Age Colliding Particles=false
    KillOnCollision=false
    Module.Control Rest State Via Surface Normal=false
    Module.EnableMaxCollisionCount=false
    Module.Randomize Collision Normal Vector=false
   ) {enabled=false}
   &SolveForcesAndVelocity(Clamp Velocity=false, Limit Acceleration=false)
   &ScaleMeshSize(Scale Factor=&VectorFromFloat(Value=&FloatFromCurve(
      FloatCurve=DataInterface() {propertyValues {Curve {keys [
          <at 0 0 RCIM_Cubic>
          {interpMode=RCIM_Cubic, tangentMode=RCTM_User, time=0.05, value=1}
          {interpMode=RCIM_Cubic, tangentMode=RCTM_User, time=0.4234, value=1}
          {arriveTangent=-1.4032, interpMode=RCIM_Cubic, leaveTangent=-1.4032, tangentMode=RCTM_User, time=0.8, value=0.05}
         ]}}}
      Scale Curve=0.5
     )))
   &ScaleMeshSizeBySpeed(Max Scale Factor=<1.5 1.5 2>, Min Scale Factor:Vector3f=<1.3 1.3 0.7>, Speed Threshold=200)
   &ScaleColor(
    Scale Alpha=&FloatFromCurve(FloatCurve=DataInterface() {propertyValues {Curve {keys [
         {tangentMode=RCTM_User, time=0, value=0}
         <at 0.0599 1>
         {arriveTangent=-0.0405, interpMode=RCIM_Cubic, leaveTangent=-0.0405, tangentMode=RCTM_User, time=0.7009, value=1}
         {arriveTangent=-8.0089, interpMode=RCIM_Cubic, leaveTangent=-8.0089, tangentMode=RCTM_User, time=1, value=0}
        ]}}})
    ScaleA=true
    ScaleRGB=true
   )
   /SetVariables(
    Particles.food Scale curve:Vector3f=&Multiply_VectorByFloat(
     Vector=&Multiply_Vector(
      A=&SelectVectorFromArray(
       Direct Array Index=$Particles.food MeshIndex
       Vector Selection Array=DataInterface() {propertyValues {FloatData [<0.35 0.35 0.35>, <0.15 0.15 0.15>, <0.35 0.35 0.35>, <0.2 0.2 0.2>, <0.1 0.1 0.1>]}}
      )
      B=<1 1 1>
     )
     Float=&FloatFromCurve(FloatCurve=DataInterface() {propertyValues {Curve {keys [
          <at 0 0.5 RCIM_Cubic>
          {interpMode=RCIM_Cubic, tangentMode=RCTM_User, time=0.05, value=1}
          {interpMode=RCIM_Cubic, tangentMode=RCTM_User, time=0.4004, value=1}
          {arriveTangent=-1.4032, interpMode=RCIM_Cubic, leaveTangent=-1.4032, tangentMode=RCTM_User, time=0.944, value=0.55}
         ]}}})
    )
    Particles.food bits Scale:Vector3f=&SelectVectorFromArray(
     Direct Array Index=$Particles.food MeshIndex
     Vector Selection Array=DataInterface() {propertyValues {FloatData [<0.35 0.35 0.35>, <0.15 0.15 0.15>, <0.35 0.35 0.35>, <0.2 0.2 0.2>, <0.1 0.1 0.1>]}}
    )
    Particles.food color:LinearColor=&MakeLinearColorFromVectorAndFloat(
     "Vector (RGB)"=<1 1 1>
     "Float (Alpha)"=&FloatFromCurve(FloatCurve=DataInterface() {propertyValues {Curve {keys [
          {tangentMode=RCTM_User, time=0, value=0}
          <at 0.0599 1>
          {arriveTangent=-0.0405, interpMode=RCIM_Cubic, leaveTangent=-0.0405, tangentMode=RCTM_User, time=0.5896, value=1}
          {arriveTangent=-8.0089, interpMode=RCIM_Cubic, leaveTangent=-8.0089, tangentMode=RCTM_User, time=1, value=0}
         ]}}})
    )
   ) {isSetParameters=true}
  ]
  rendererClasses [&NiagaraMeshRendererProperties]
  renderers [
   {rendererClass=&NiagaraMeshRendererProperties, rendererIndex=0}
   {rendererClass=&NiagaraMeshRendererProperties, rendererIndex=1}
   {rendererClass=&NiagaraMeshRendererProperties, rendererIndex=2}
  ]
 }
}
summary {
 systemName=NS_Emote_FluteLamp_FoodCrumbs
 emitters [
  {emitterName=bread_crrumbs, rendererClasses [&NiagaraMeshRendererProperties]}
  {emitterName=goopy, rendererClasses [&NiagaraMeshRendererProperties]}
  {emitterName=Sparkles001, rendererClasses [&NiagaraSpriteRendererProperties]}
  {emitterName=goopy002, rendererClasses [&NiagaraMeshRendererProperties]}
  {bEnabled=false, emitterName=Mouth_goop_spray_v01, rendererClasses [&NiagaraMeshRendererProperties]}
  {emitterName=Mouth_goop_spray_v2, rendererClasses [&NiagaraMeshRendererProperties]}
 ]
 userVariables []
}
