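# Niagara one-emitter effect for the FluteLamp emote's sandwich squeeze, throwing out short-lived mesh blobs, apparently ketchup and mustard given its color parameters.
# Two instant bursts use different spawn groups, copied to VisibilityTag, presumably to split particles between two mesh renderers.
# Particles spawn in a thin cylinder with randomized negative point velocity, then fall under gravity and drag.
# Speed-based stretching plus size and alpha curves shape each blob over its life.
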NiagaraSystem /BRCosmetics/Animation/Game/MainPlayer/Emotes/FluteLamp/FX/NS_Emote_FluteLamp_Sandwich_squeeze
dependencies {
 usedDataInterfaces [&NiagaraDataInterfaceCurve, &NiagaraDataInterfaceSocketReader]
 usedDynamicInputs []
 usedModules [
  &SystemState
  &self:SystemSpawnScript.SystemScriptSource.SystemScriptGraph.NiagaraNodeAssignment_0.SetVariables_#1
 ]
}
emitters {Big_Squeeze {
  emitterUpdateScript [
   &EmitterState()
   &SpawnBurst_Instantaneous(Spawn Count=25, Spawn Group=1, Use Spawn Probability=false, UseLoopCountLimit=false) {module=SpawnBurst_Instantaneous001}
   &SpawnBurst_Instantaneous(Spawn Count=15, Use Spawn Probability=false, UseLoopCountLimit=false)
   /SetVariables(
    Emitter.Ketchup 1:LinearColor=<rgba 0.1448 0.0058 0.0024 1>
    Emitter.Ketchup 2:LinearColor=<rgba 1 0 0 1>
    Emitter.Mustard 2:LinearColor=<rgba 1 0.59 0 1>
    Emitter.mustard 1:LinearColor=<rgba 0.5323 0.3141 0 1>
   ) {isSetParameters=true}
  ]
  particleSpawnScript [
   /SetVariables(Particles.VisibilityTag=&SpawnGroup()) {isSetParameters=true}
   &InitializeParticle(
    Lifetime Max=0.5
    Lifetime Min=0.3
    Lifetime Mode=::NewEnumerator1
    Mesh Scale Max=<0.5 2 4>
    Mesh Scale Min=<0.2 0.25 1>
    Mesh Scale Mode=::NewEnumerator2
    Sprite Size Max=<3 8>
    Sprite Size Min=<1 5>
    Sprite Size Mode=::NewEnumerator2
    UsePositionOffset=false
    Write Material Random=true
   )
   &InitialMeshOrientation(Mesh Orientation Mode=::NewEnumerator7, Rotation=<0 -0.25 0>, Use Rotation Vector=true)
   &DynamicMaterialParameters(
    Index 0 Param 1=1
    Index 0 Param 2=1
    Module.Param0WriteEnabled=true
    Module.Param1WriteEnabled=true
    Module.Param2WriteEnabled=false
    Module.Param3WriteEnabled=false
   )
   &AddVelocityFromPoint(Module.Use Falloff Distance=false, Velocity Strength=&RandomRangeFloat(Minimum=-200, Maximum=-300))
   &ShapeLocation(
    Cylinder Height=37
    Cylinder Radius=2
    MaskSpawnGroup=false
    Override Seed=false
    Shape Primitive=::NewEnumerator1
    Surface Only=false
    UseOwnerScale=true
   )
  ]
  particleUpdateScript [
   &ParticleState()
   &GravityForce(Gravity=&RandomRangeVector(Minimum=<0 0 -200>, Maximum=<0 0 -400>))
   &Drag(Drag=&RandomRangeFloat(Minimum=2, Maximum=3), Use Linear Drag=true, Use Rotational Drag=false)
   &SolveForcesAndVelocity(Clamp Velocity=false, Limit Acceleration=false)
   &ScaleMeshSize(Scale Factor=&VectorFromFloat(Value=&FloatFromCurve(
      FloatCurve=DataInterface() {propertyValues {Curve {keys [
          <at 0 0 RCIM_Cubic>
          {interpMode=RCIM_Cubic, tangentMode=RCTM_User, time=0.2169, value=1}
          {interpMode=RCIM_Cubic, tangentMode=RCTM_User, time=0.5, value=1}
          {arriveTangent=-1.4032, interpMode=RCIM_Cubic, leaveTangent=-1.4032, tangentMode=RCTM_User, time=0.873, value=0}
         ]}}}
      Scale Curve=0.5
     )))
   &ScaleMeshSizeBySpeed(
    Max Scale Factor=&RandomRangeVector(Minimum=<1 1 2>, Maximum=<1.5 1.5 5>)
    Min Scale Factor:Vector3f=<1 1 0.33>
    Sample Scale Factor By Curve=true
    Scale Factor Curve=DataInterface() {propertyValues {Curve {keys [
        {interpMode=RCIM_Cubic, leaveTangent=-0.1696, tangentMode=RCTM_Break, time=0, value=0}
        {interpMode=RCIM_Cubic, leaveTangent=-0.1696, tangentMode=RCTM_Break, time=0.5, value=0.75}
        {interpMode=RCIM_Cubic, leaveTangent=-0.1696, tangentMode=RCTM_Break, time=1, value=0}
       ]}}}
    Speed Threshold=250
   )
   &ScaleColor(
    Scale Alpha=&FloatFromCurve(FloatCurve=DataInterface() {propertyValues {Curve {keys [
         {tangentMode=RCTM_User, time=0, value=0}
         <at 0.0599 1>
         {arriveTangent=-0.0405, interpMode=RCIM_Cubic, leaveTangent=-0.0405, tangentMode=RCTM_User, time=0.5875, value=1}
         {arriveTangent=-8.0089, interpMode=RCIM_Cubic, leaveTangent=-8.0089, tangentMode=RCTM_User, time=0.8867, value=0}
        ]}}})
    ScaleA=true
    ScaleRGB=true
   )
  ]
  rendererClasses [&NiagaraMeshRendererProperties]
  renderers [
   {rendererClass=&NiagaraMeshRendererProperties, rendererIndex=0}
   {rendererClass=&NiagaraMeshRendererProperties, rendererIndex=1}
  ]
 }}
summary {
 systemName=NS_Emote_FluteLamp_Sandwich_squeeze
 emitters [{emitterName=Big_Squeeze, rendererClasses [&NiagaraMeshRendererProperties]}]
 userVariables [
  User.bIsFrontEnd {type=NiagaraBool}
  User.bIsFrontEndPreview {type=NiagaraBool}
  User.DripSpawnRate {type=NiagaraFloat, defaultValue=50, description=""}
 ]
}
