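# Niagara effect for the FluteLamp emote that throws ketchup, mustard and food-bit meshes amid dark, swirling ribbon trails.
# One combined emitter copies spawn groups into VisibilityTag, switching lifetime, scale, velocity and forces per debris type; the separate food and condiment emitters are disabled.
# Short-lived curl-noise sprites feed the trails emitter through a particle attribute reader, driving ribbon position and width, with time-of-day tint and alpha.
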NiagaraSystem /BRCosmetics/Animation/Game/MainPlayer/Emotes/FluteLamp/FX/NS_Emote_FluteLamp_foodbits
dependencies {
 usedDataInterfaces [
  &NiagaraDataInterfaceArrayInt32
  &NiagaraDataInterfaceCurve
  &NiagaraDataInterfaceVectorCurve
  &NiagaraDataInterfaceParticleRead
 ]
 usedDynamicInputs [&Multiply_Vector2D_ByFloat]
 usedModules [&SystemState]
}
emitters {
 Food_Bits {
  bEnabled=false
  emitterUpdateScript [
   &EmitterState()
   &SpawnRate(Spawn Probability=0.5, SpawnRate=35, Use Spawn Probability=true)
  ]
  particleSpawnScript [
   &InitializeParticle(
    Color=<rgba 1 1 1 0.5>
    Lifetime Max=1
    Lifetime Min=0.4
    Lifetime Mode=::NewEnumerator1
    Mesh Scale Max=<0.2 0.2 0.2>
    Mesh Scale Min=<0.15 0.15 0.15>
    Mesh Scale Mode=::NewEnumerator2
    UsePositionOffset=false
    Write Material Random=true
   )
   /SetVariables(Particles.MeshIndex=&SelectIntFromArray(
     Array Sampling Mode=::NewEnumerator1
     Direct Array Index=?
     Int Selection Array=DataInterface() {propertyValues {IntData [0,1,2,3,4]}}
     Random Seed=0
    )) {isSetParameters=true}
   &AddVelocityFromPoint(Module.Use Falloff Distance=false, Velocity Strength=&RandomRangeFloat(Minimum=-300, Maximum=350))
  ]
  particleUpdateScript [
   &ParticleState()
   &UpdateMeshOrientation(Rotation Vector=&RandomVector(Vector Scale=1))
   &ScaleMeshSize(Scale Factor=&VectorFromFloat(Value=&FloatFromCurve(
      FloatCurve=DataInterface() {propertyValues {Curve {keys [
          <at 0 0 RCIM_Cubic>
          {interpMode=RCIM_Cubic, tangentMode=RCTM_User, time=0.25, value=1}
          {interpMode=RCIM_Cubic, tangentMode=RCTM_User, time=0.5, value=1}
          {arriveTangent=-1.4032, interpMode=RCIM_Cubic, leaveTangent=-1.4032, tangentMode=RCTM_User, time=1, value=0.5}
         ]}}}
      Scale Curve=0.8
     )))
   &Drag(Drag=&RandomRangeFloat(Minimum=0, Maximum=1), Use Linear Drag=true, Use Rotational Drag=false)
   &GravityForce(Gravity=&RandomRangeVector(Minimum=<0 0 -200>, Maximum=<0 0 -300>))
   &MeshRotationForce(Rotation=<1 0 5>)
   &SolveForcesAndVelocity(Clamp Velocity=false, Limit Acceleration=false)
   &ScaleMeshSizeBySpeed(Max Scale Factor=<1 1 2>, Min Scale Factor:Vector3f=<1 1 0.33>, Speed Threshold=200)
   &ScaleColor(
    Scale Alpha=&FloatFromCurve(FloatCurve=DataInterface() {propertyValues {Curve {keys [
         {tangentMode=RCTM_User, time=0, value=0}
         <at 0.1907 1>
         {arriveTangent=-0.0405, interpMode=RCIM_Cubic, leaveTangent=-0.0405, tangentMode=RCTM_User, time=0.67, value=1}
         {arriveTangent=-8.0089, interpMode=RCIM_Cubic, leaveTangent=-8.0089, tangentMode=RCTM_User, time=1, value=0}
        ]}}})
    ScaleA=true
    ScaleRGB=true
   ) {enabled=false}
  ]
  rendererClasses [&NiagaraMeshRendererProperties]
  renderers [{rendererClass=&NiagaraMeshRendererProperties, rendererIndex=0}]
 }
 SmokeLWE001 {rendererClasses [], renderers []}
 SmokeLWE_center {rendererClasses [], renderers []}
 chaotic_particles {
  emitterUpdateScript [
   &EmitterState(Life Cycle Mode=::NewEnumerator1, Loop Behavior=::NewEnumerator1, Loop Duration=2, UseLoopDelay=false)
   &SpawnRate(SpawnRate=13, Use Spawn Probability=false)
  ]
  particleSpawnScript [
   &InitializeParticle(
    Lifetime Max=0.6
    Lifetime Min=0.35
    Lifetime Mode=::NewEnumerator1
    Position Mode=::NewEnumerator0
    Sprite Size Mode=::NewEnumerator3
    Uniform Sprite Size=0.05
    Write Material Random=true
   )
   &AddVelocity(Module.Use Falloff Distance=false, Override Seed=false, Velocity Mode=::NewEnumerator1, Velocity Speed=100) {enabled=false}
  ]
  particleUpdateScript [
   &ParticleState()
   &ScaleSpriteSize(
    Uniform Curve Scale=&Lerp_Float(
     A=1
     B=1.8
     Alpha=&FN_LOD_Distance_Remap(Start Distance=100, End Distance=1000, Minimum Output=0, Maximum Output=1, Clamp Output=true)
    )
    Uniform Curve Sprite Scale=DataInterface() {propertyValues {Curve {defaultValue=0, keys [{tangentMode=RCTM_User, time=0, value=0.4118}, <at 0.2094 1>, <at 0.323 1>, {tangentMode=RCTM_User, time=1, value=0}]}}}
   )
   &CurlNoiseForce(
    Bias Noise=false
    Noise Frequency=120
    Noise Strength=&FloatFromCurve(
     FloatCurve=DataInterface() {propertyValues {Curve {defaultValue=1, keys [{tangentMode=RCTM_User, time=0, value=0}, {tangentMode=RCTM_User, time=1, value=1}]}}}
     Scale Curve=&RandomRangeFloat(Minimum=2000, Maximum=4000)
    )
    Pan Noise=false
   )
   &ShapeLocation(
    MaskSpawnGroup=false
    Offset Coordinate Space=::World
    Offset Mode=::NewEnumerator0
    Rotation Angle:NiagaraFloat=&RandomRangeFloat(Minimum=0, Maximum=1)
    Rotation Axis=<1 1 1>
    Rotation Coordinate Space=::World
    Rotation Mode=::NewEnumerator1
    Sphere Distribution=::NewEnumerator1
    Sphere Radius=40
    Sphere Radius Position:NiagaraFloat=&RandomRangeFloat(Minimum=0.5, Maximum=1)
    U Position=&MakeCustomFloatFromBool(
     Bool=&RandomBool()
     True Float=&FloatFromCurve(
      FloatCurve=DataInterface() {propertyValues {Curve {
         defaultValue=1
         keys [
          {arriveTangent=4.5812, interpMode=RCIM_Cubic, leaveTangent=4.5812, tangentMode=RCTM_User, time=0, value=0}
          {arriveTangent=0.9143, interpMode=RCIM_Cubic, leaveTangent=0.9143, tangentMode=RCTM_User, time=0.2561, value=0.7055}
          {arriveTangent=0.0224, interpMode=RCIM_Cubic, leaveTangent=0.0224, tangentMode=RCTM_User, time=1.0791, value=1}
         ]
        }}}
      CurveIndex=&Multiply_Float(A=$Particles.NormalizedAge, B=&RandomRangeFloat(Minimum=0.1, Maximum=0.8))
      Scale Curve=&RandomRangeFloat(Minimum=1, Maximum=2.2)
     )
     False Float=&FloatFromCurve(
      FloatCurve=DataInterface() {propertyValues {Curve {keys [{interpMode=RCIM_Cubic, leaveTangent=-1.7692, tangentMode=RCTM_Break, time=0, value=1}, <at 1 0>]}}}
      CurveIndex=&Multiply_Float(A=$Particles.NormalizedAge, B=&RandomRangeFloat(Minimum=0.25, Maximum=0.7))
      Scale Curve=&RandomRangeFloat(Minimum=1, Maximum=2)
     )
    )
    UseOwnerScale=true
    V Position=&MakeCustomFloatFromBool(
     Bool=&RandomBool()
     True Float=&FloatFromCurve(
      FloatCurve=DataInterface() {propertyValues {Curve {
         defaultValue=1
         keys [
          {arriveTangent=4.5812, interpMode=RCIM_Cubic, leaveTangent=4.5812, tangentMode=RCTM_User, time=0, value=0}
          {arriveTangent=1.6567, interpMode=RCIM_Cubic, leaveTangent=1.6567, tangentMode=RCTM_User, time=0.1811, value=0.6464}
          {arriveTangent=0.0224, interpMode=RCIM_Cubic, leaveTangent=0.0224, tangentMode=RCTM_User, time=1, value=1}
         ]
        }}}
      CurveIndex=&Multiply_Float(A=$Particles.NormalizedAge, B=&RandomRangeFloat(Minimum=0.5, Maximum=1))
      Scale Curve=1.35
     )
     False Float=&FloatFromCurve(
      FloatCurve=DataInterface() {propertyValues {Curve {keys [
          {interpMode=RCIM_Cubic, leaveTangent=-3.3122, tangentMode=RCTM_Break, time=0, value=1}
          {arriveTangent=-0.4912, tangentMode=RCTM_User, time=1, value=0}
         ]}}}
      CurveIndex=&Multiply_Float(A=$Particles.NormalizedAge, B=&RandomRangeFloat(Minimum=0.1, Maximum=0.5))
     )
    )
   ) {module=ShapeLocation003}
   &ScaleVelocity(Velocity Scale=&VectorFromCurve(
     VectorCurve=DataInterface() {propertyValues {
       XCurve {keys [
         {interpMode=RCIM_Cubic, tangentMode=RCTM_Break, time=0, value=0.2}
         {interpMode=RCIM_Cubic, tangentMode=RCTM_Break, time=0.5, value=1}
         <at 1 0.2>
        ]}
       YCurve {keys [
         {interpMode=RCIM_Cubic, tangentMode=RCTM_Break, time=0, value=0.2}
         {interpMode=RCIM_Cubic, tangentMode=RCTM_Break, time=0.5, value=1}
         <at 1 0.2>
        ]}
       ZCurve {keys [
         {interpMode=RCIM_Cubic, tangentMode=RCTM_Break, time=0, value=0.2}
         {arriveTangent=0.4726, interpMode=RCIM_Cubic, tangentMode=RCTM_Break, time=0.5, value=1}
         <at 1 0.2>
        ]}
      }}
     Scale Curve=&RandomRangeVector(Minimum=<1 1 1>, Maximum=<2 2 2>)
    ))
   &Drag(Drag=&RandomRangeFloat(Minimum=1, Maximum=2), Use Linear Drag=true, Use Rotational Drag=false) {enabled=false}
   &SolveForcesAndVelocity(Clamp Velocity=false, Limit Acceleration=false)
  ]
  rendererClasses [&NiagaraSpriteRendererProperties]
  renderers [{rendererClass=&NiagaraSpriteRendererProperties}]
 }
 ketchup_and_mustard001 {
  bEnabled=false
  emitterUpdateScript [
   &EmitterState()
   &SpawnRate(Spawn Probability=0.1, SpawnGroup=2, SpawnRate=50, Use Spawn Probability=true)
   &SpawnRate(Spawn Probability=0.5, SpawnGroup=1, SpawnRate=25, Use Spawn Probability=true) {module=SpawnRate002}
   &SpawnRate(SpawnRate=15, Use Spawn Probability=true) {module=SpawnRate001}
   /SetVariables(
    Emitter.Ketchup 1:LinearColor=<rgba 0.1448 0.0058 0.0024 1>
    Emitter.Ketchup 2:LinearColor=<rgba 1 0 0 1>
    Emitter.Mustard 2:LinearColor=<rgba 1 0.59 0 1>
    Emitter.mustard 1:LinearColor=<rgba 0.5323 0.3141 0 1>
   ) {isSetParameters=true, module=SetVariables_#1}
  ]
  particleSpawnScript [
   /SetVariables(Particles.VisibilityTag=&SpawnGroup()) {isSetParameters=true, module=SetVariables_#2}
   &InitializeParticle(
    Color=&MergeHelper_ColorSwitch(0=<rgba 1 1 1 1>, 1=<rgba 1 1 1 1>, 2=<rgba 1 1 1 0.5>, 3=?, Index=$Particles.VisibilityTag, Num=::NewEnumerator1)
    Lifetime Max=0.5
    Lifetime Min=0.3
    Lifetime Mode=::NewEnumerator1
    Mesh Scale Max=<0.5 2 4>
    Mesh Scale Min=<0.2 0.25 1>
    Mesh Scale Mode=::NewEnumerator2
    Sprite Size Max=<3 8>
    Sprite Size Min=<1 5>
    Sprite Size Mode=::NewEnumerator2
    UsePositionOffset=false
    Write Material Random=true
   )
   &InitialMeshOrientation(Mesh Orientation Mode=::NewEnumerator7, Rotation=<0 -0.25 0>, Use Rotation Vector=true)
   &DynamicMaterialParameters(
    Index 0 Param 1=1
    Index 0 Param 2=1
    Module.Param0WriteEnabled=true
    Module.Param1WriteEnabled=true
    Module.Param2WriteEnabled=false
    Module.Param3WriteEnabled=false
   )
   &AddVelocityFromPoint(
    Module.Use Falloff Distance=false
    Velocity Strength=&MergeHelper_FloatSwitch(
     0=&RandomRangeFloat(Minimum=-150, Maximum=-250)
     1=&RandomRangeFloat(Minimum=-150, Maximum=-250)
     2=&RandomRangeFloat(Minimum=-100, Maximum=350)
     3=?
     Index=$Particles.VisibilityTag
     Num=::NewEnumerator1
    )
   )
   &ShapeLocation(
    Cylinder Height=37
    Cylinder Radius=2
    MaskSpawnGroup=false
    Override Seed=false
    Shape Primitive=::NewEnumerator1
    Surface Only=false
    UseOwnerScale=true
   )
  ]
  particleUpdateScript [
   &ParticleState()
   &GravityForce(Gravity=&RandomRangeVector(Minimum=<0 0 -200>, Maximum=<0 0 -400>))
   &Drag(Drag=&RandomRangeFloat(Minimum=0, Maximum=1.5), Use Linear Drag=true, Use Rotational Drag=false)
   &SolveForcesAndVelocity(Clamp Velocity=false, Limit Acceleration=false)
   &ScaleMeshSize(Scale Factor=&VectorFromFloat(Value=&FloatFromCurve(
      FloatCurve=DataInterface() {propertyValues {Curve {keys [
          <at 0 0 RCIM_Cubic>
          {interpMode=RCIM_Cubic, tangentMode=RCTM_User, time=0.2169, value=1}
          {interpMode=RCIM_Cubic, tangentMode=RCTM_User, time=0.5, value=1}
          {arriveTangent=-1.4032, interpMode=RCIM_Cubic, leaveTangent=-1.4032, tangentMode=RCTM_User, time=1, value=0.5}
         ]}}}
      Scale Curve=0.5
     )))
   &ScaleMeshSizeBySpeed(
    Max Scale Factor=&MergeHelper_VectorSwitch(
     0=&RandomRangeVector(Minimum=<1 1 2>, Maximum=<1 1 5>)
     1=&RandomRangeVector(Minimum=<1 1 2>, Maximum=<1 1 5>)
     2=<1 1 2>
     3=?
     Index=$Particles.VisibilityTag
     Num=::NewEnumerator1
    )
    Min Scale Factor:Vector3f=&MergeHelper_VectorSwitch(0=<1 1 0.33>, 1=<1 1 0.33>, 2=<1 1 0.33>, 3=?, Index=$Particles.VisibilityTag, Num=::NewEnumerator1)
    Sample Scale Factor By Curve=true
    Scale Factor Curve=DataInterface() {propertyValues {Curve {keys [
        {interpMode=RCIM_Cubic, leaveTangent=-0.1696, tangentMode=RCTM_Break, time=0, value=0}
        {interpMode=RCIM_Cubic, leaveTangent=-0.1696, tangentMode=RCTM_Break, time=0.5, value=0.75}
        {interpMode=RCIM_Cubic, leaveTangent=-0.1696, tangentMode=RCTM_Break, time=1, value=0}
       ]}}}
    Speed Threshold:NiagaraFloat=&MergeHelper_FloatSwitch(0=250, 1=250, 2=200, 3=?, Index=$Particles.VisibilityTag, Num=::NewEnumerator1)
   ) {module=ScaleMeshSizeBySpeed001}
   &ScaleColor(
    Scale Alpha=&FloatFromCurve(FloatCurve=DataInterface() {propertyValues {Curve {keys [
         {tangentMode=RCTM_User, time=0, value=0}
         <at 0.0599 1>
         {arriveTangent=-0.0405, interpMode=RCIM_Cubic, leaveTangent=-0.0405, tangentMode=RCTM_User, time=0.7009, value=1}
         {arriveTangent=-8.0089, interpMode=RCIM_Cubic, leaveTangent=-8.0089, tangentMode=RCTM_User, time=1, value=0}
        ]}}})
    ScaleA=true
    ScaleRGB=true
   )
  ]
  rendererClasses [&NiagaraMeshRendererProperties]
  renderers [
   {rendererClass=&NiagaraMeshRendererProperties, rendererIndex=0}
   {rendererClass=&NiagaraMeshRendererProperties, rendererIndex=1}
   {rendererClass=&NiagaraMeshRendererProperties, rendererIndex=2}
  ]
 }
 ketchup_mustard_and_food {
  emitterUpdateScript [
   &EmitterState()
   &SpawnRate(Spawn Probability=0.5, SpawnGroup=3, SpawnRate=35, Use Spawn Probability=true) {module=SpawnRate003}
   &SpawnRate(Spawn Probability=0.1, SpawnGroup=2, SpawnRate=50, Use Spawn Probability=true)
   &SpawnRate(Spawn Probability=0.5, SpawnGroup=1, SpawnRate=25, Use Spawn Probability=true) {module=SpawnRate002}
   &SpawnRate(SpawnRate=15, Use Spawn Probability=true) {module=SpawnRate001}
   /SetVariables(
    Emitter.Ketchup 1:LinearColor=<rgba 0.1448 0.0058 0.0024 1>
    Emitter.Ketchup 2:LinearColor=<rgba 1 0 0 1>
    Emitter.Mustard 2:LinearColor=<rgba 1 0.59 0 1>
    Emitter.mustard 1:LinearColor=<rgba 0.5323 0.3141 0 1>
   ) {isSetParameters=true, module=SetVariables_#1}
  ]
  particleSpawnScript [
   /SetVariables(Particles.VisibilityTag=&SpawnGroup()) {isSetParameters=true, module=SetVariables_#2}
   &InitializeParticle(
    Color=&MergeHelper_ColorSwitch(0=<rgba 1 1 1 1>, 1=<rgba 1 1 1 1>, 2=<rgba 1 1 1 0.5>, 3=?, Index=$Particles.VisibilityTag, Num=::NewEnumerator1)
    Lifetime=&MergeHelper_FloatSwitch(
     0=&RandomRangeFloat(Minimum=0.3, Maximum=0.5)
     1=&RandomRangeFloat(Minimum=0.3, Maximum=0.5)
     2=&RandomRangeFloat(Minimum=0.2, Maximum=0.55)
     3=&RandomRangeFloat(Minimum=0.4, Maximum=1)
     Index=$Particles.VisibilityTag
     Num=::NewEnumerator2
    )
    Mesh Scale=&MergeHelper_VectorSwitch(
     0=&RandomRangeVector(Minimum=<0.2 0.25 1>, Maximum=<0.5 2 4>)
     1=&RandomRangeVector(Minimum=<0.2 0.25 1>, Maximum=<0.5 2 4>)
     2=&RandomRangeVector(Minimum=<0.2 0.25 1>, Maximum=<0.5 2 4>)
     3=&RandomRangeVector(Minimum=<0.15 0.15 0.15>, Maximum=<0.2 0.2 0.2>)
     Index=$Particles.VisibilityTag
     Num=::NewEnumerator2
    )
    Mesh Scale Mode=::NewEnumerator4
    Sprite Size Max=<3 8>
    Sprite Size Min=<1 5>
    Sprite Size Mode=::NewEnumerator2
    UsePositionOffset=false
    Write Material Random=true
   )
   &InitialMeshOrientation(Mesh Orientation Mode=::NewEnumerator7, Rotation=<0 -0.25 0>, Use Rotation Vector=true)
   &DynamicMaterialParameters(
    Index 0 Param 1=1
    Index 0 Param 2=1
    Module.Param0WriteEnabled=true
    Module.Param1WriteEnabled=true
    Module.Param2WriteEnabled=false
    Module.Param3WriteEnabled=false
   )
   &AddVelocityFromPoint(
    Module.Use Falloff Distance=false
    Velocity Strength=&MergeHelper_FloatSwitch(
     0=&RandomRangeFloat(Minimum=-150, Maximum=-250)
     1=&RandomRangeFloat(Minimum=-150, Maximum=-250)
     2=&RandomRangeFloat(Minimum=-100, Maximum=350)
     3=&RandomRangeFloat(Minimum=-300, Maximum=350)
     Index=$Particles.VisibilityTag
     Num=::NewEnumerator2
    )
   )
   &ShapeLocation(
    Cylinder Height=37
    Cylinder Radius=2
    MaskSpawnGroup=false
    Override Seed=false
    Shape Primitive=::NewEnumerator1
    Surface Only=false
    UseOwnerScale=true
   )
   /SetVariables(Particles.Food MeshIndex:NiagaraInt32=&SelectIntFromArray(
     Array Sampling Mode=::NewEnumerator1
     Direct Array Index=?
     Int Selection Array=DataInterface() {propertyValues {IntData [0,1,2,3,4]}}
     Random Seed=0
    )) {isSetParameters=true}
  ]
  particleUpdateScript [
   &ParticleState()
   &UpdateMeshOrientation(
    Rotation Coordinate Space=::NewEnumerator2
    Rotation Vector=&RandomVector(Vector Scale=&MergeHelper_FloatSwitch(0=0, 1=0, 2=0, 3=1, Index=$Particles.VisibilityTag, Num=::NewEnumerator2))
   )
   &GravityForce(Gravity=&MergeHelper_VectorSwitch(
     0=&RandomRangeVector(Minimum=<0 0 -200>, Maximum=<0 0 -410>)
     1=&RandomRangeVector(Minimum=<0 0 -200>, Maximum=<0 0 -400>)
     2=&RandomRangeVector(Minimum=<0 0 -200>, Maximum=<0 0 -390>)
     3=&RandomRangeVector(Minimum=<0 0 -200>, Maximum=<0 0 -300>)
     Index=$Particles.VisibilityTag
     Num=::NewEnumerator2
    ))
   &Drag(
    Drag=&RandomRangeFloat(
     Minimum=0
     Maximum=&MergeHelper_FloatSwitch(0=1.5, 1=1.5, 2=1.5, 3=1, Index=$Particles.VisibilityTag, Num=::NewEnumerator2)
    )
    Use Linear Drag=true
    Use Rotational Drag=false
   )
   &MeshRotationForce(Rotation=&MergeHelper_VectorSwitch(0=<0 0 0>, 1=<0 0 0>, 2=<0 0 0>, 3=<1 0 5>, Index=$Particles.VisibilityTag, Num=::NewEnumerator2))
   &SolveForcesAndVelocity(Clamp Velocity=false, Limit Acceleration=false)
   &ScaleMeshSize(Scale Factor=&VectorFromFloat(Value=&FloatFromCurve(
      FloatCurve=DataInterface() {propertyValues {Curve {keys [
          <at 0 0 RCIM_Cubic>
          {interpMode=RCIM_Cubic, tangentMode=RCTM_User, time=0.2169, value=1}
          {interpMode=RCIM_Cubic, tangentMode=RCTM_User, time=0.5, value=1}
          {arriveTangent=-1.4032, interpMode=RCIM_Cubic, leaveTangent=-1.4032, tangentMode=RCTM_User, time=1, value=0.5}
         ]}}}
      Scale Curve=&MergeHelper_FloatSwitch(0=0.5, 1=0.5, 2=0.5, 3=0.8, Index=$Particles.VisibilityTag, Num=::NewEnumerator2)
     )))
   &ScaleMeshSizeBySpeed(
    Max Scale Factor=&MergeHelper_VectorSwitch(
     0=&RandomRangeVector(Minimum=<1 1 2>, Maximum=<1 1 5>)
     1=&RandomRangeVector(Minimum=<1 1 2>, Maximum=<1 1 5>)
     2=<1 1 2>
     3=<1 1 2>
     Index=$Particles.VisibilityTag
     Num=::NewEnumerator2
    )
    Min Scale Factor:Vector3f=&MergeHelper_VectorSwitch(0=<1 1 0.33>, 1=<1 1 0.33>, 2=<1 1 0.33>, 3=<1 1 0.33>, Index=$Particles.VisibilityTag, Num=::NewEnumerator2)
    Sample Scale Factor By Curve=true
    Scale Factor Curve=DataInterface() {propertyValues {Curve {keys [
        {interpMode=RCIM_Cubic, leaveTangent=-0.1696, tangentMode=RCTM_Break, time=0, value=0}
        {interpMode=RCIM_Cubic, leaveTangent=-0.1696, tangentMode=RCTM_Break, time=0.5, value=0.75}
        {interpMode=RCIM_Cubic, leaveTangent=-0.1696, tangentMode=RCTM_Break, time=1, value=0}
       ]}}}
    Speed Threshold:NiagaraFloat=&MergeHelper_FloatSwitch(0=250, 1=250, 2=200, 3=200, Index=$Particles.VisibilityTag, Num=::NewEnumerator2)
   ) {module=ScaleMeshSizeBySpeed001}
   &ScaleColor(
    Scale Alpha=&FloatFromCurve(FloatCurve=DataInterface() {propertyValues {Curve {keys [
         {tangentMode=RCTM_User, time=0, value=0}
         <at 0.0599 1>
         {arriveTangent=-0.0405, interpMode=RCIM_Cubic, leaveTangent=-0.0405, tangentMode=RCTM_User, time=0.7009, value=1}
         {arriveTangent=-8.0089, interpMode=RCIM_Cubic, leaveTangent=-8.0089, tangentMode=RCTM_User, time=1, value=0}
        ]}}})
    ScaleA=true
    ScaleRGB=true
   )
  ]
  rendererClasses [&NiagaraMeshRendererProperties]
  renderers [
   {rendererClass=&NiagaraMeshRendererProperties, rendererIndex=0}
   {rendererClass=&NiagaraMeshRendererProperties, rendererIndex=1}
   {rendererClass=&NiagaraMeshRendererProperties, rendererIndex=2}
   {rendererClass=&NiagaraMeshRendererProperties, rendererIndex=3}
  ]
 }
 trails {
  emitterUpdateScript [
   &EmitterState()
   &SpawnParticlesFromOtherEmitter(
    Attribute Reader=DataInterface() {propertyValues {EmitterBinding {bindingMode=Other, emitterName=chaotic_particles}}}
    Maximum Age Spawn Threshold=1.4
    Spawn Rate=16.0307
    Use Maximum Age=false
    Use Minimum Age=false
   )
  ]
  particleSpawnScript [
   &InitializeParticle(
    Color=&FNTimeOfDayColor(
     Night Color=&RandomRangeLinearColor(Minimum=<rgba 0.024 0.024 0.024 0.4>, Maximum=<rgba 0 0 0 0.55>)
     Morning Color=&RandomRangeLinearColor(Minimum=<rgba 0.0365 0.0365 0.0365 0.4>, Maximum=<rgba 0.024 0.024 0.024 0.4>)
     Day Color=&RandomRangeLinearColor(Minimum=<rgba 0.0656 0.0656 0.0656 0.7>, Maximum=<rgba 0.024 0.024 0.024 0.3>)
     Evening Color=&RandomRangeLinearColor(Minimum=<rgba 0.024 0.024 0.024 0.4>, Maximum=<rgba 0.0005 0.0005 0.0073 0.55>)
    )
    Lifetime Max=0.7
    Lifetime Min=0.3
    Lifetime Mode=::NewEnumerator1
    Ribbon Width=&RandomRangeFloat(Minimum=200, Maximum=300)
    Ribbon Width Mode=::NewEnumerator1
    UsePositionOffset=false
    Write Material Random=true
   )
   &SampleParticlesFromOtherEmitter(
    Particle ID Sampling=::NewEnumerator1
    Reject Particle If Sampled Index Is Invalid=false
    Sprite Size Sampling:ENiagara_DirectReadSamplingMode=::NewEnumerator1
   )
  ]
  particleUpdateScript [
   &ParticleState()
   /Niagara/Modules/AttributeReader/UpdateParticlesFromOtherEmitter(
    Attribute Reader=DataInterface() {propertyValues {EmitterBinding {bindingMode=Other, emitterName=chaotic_particles}}}
    Position Sampling:ENiagara_DirectReadSamplingMode=::NewEnumerator3
    Sprite Size Sampling:ENiagara_DirectReadSamplingMode=::NewEnumerator1
   )
   &ScaleRibbonWidth(Ribbon Width Scale=&Multiply_Float(
     A=&FloatFromCurve(FloatCurve=DataInterface() {propertyValues {Curve {
         defaultValue=1
         keys [
          {interpMode=RCIM_Cubic, tangentMode=RCTM_User, time=0, value=0}
          {arriveTangent=0.1693, interpMode=RCIM_Cubic, leaveTangent=0.1693, tangentMode=RCTM_User, time=0.2, value=1}
          {interpMode=RCIM_Cubic, tangentMode=RCTM_User, time=0.5, value=1}
          {tangentMode=RCTM_User, time=1, value=0}
         ]
        }}})
     B=&MakeFloatFromVector2d(Vector 2D=$Output.UpdateParticlesFromOtherEmitter.SampledSpriteSize)
    ))
   &InheritSourceMovement(Applied Position Scale=<0.2 0.2 0.2>, ApplyForce=false, ApplyPosition=true, ApplyPositionFalloff=false)
   &SolveForcesAndVelocity(Clamp Velocity=false, Limit Acceleration=false)
   &ScaleColor(
    Scale Alpha=&FloatFromCurve(
     FloatCurve=DataInterface() {propertyValues {Curve {
        defaultValue=1
        keys [{interpMode=RCIM_Cubic, tangentMode=RCTM_User, time=0, value=1}, {interpMode=RCIM_Cubic, tangentMode=RCTM_User, time=0.85, value=0}]
       }}}
     Scale Curve=&FNTimeOfDayScalar(Night Scale=7, Morning Scale=10, Day Scale=10, Evening Scale=15)
    )
    ScaleA=true
    ScaleRGB=true
   )
  ]
  rendererClasses [&NiagaraRibbonRendererProperties]
  renderers [
   {rendererClass=&NiagaraRibbonRendererProperties}
   {rendererClass=&NiagaraRibbonRendererProperties, rendererIndex=1}
  ]
 }
}
summary {
 systemName=NS_Emote_FluteLamp_foodbits
 emitters [
  {bEnabled=false, emitterName=Food_Bits, rendererClasses [&NiagaraMeshRendererProperties]}
  {emitterName=SmokeLWE_center, rendererClasses []}
  {bEnabled=false, emitterName=ketchup_and_mustard001, rendererClasses [&NiagaraMeshRendererProperties]}
  {emitterName=chaotic_particles, rendererClasses [&NiagaraSpriteRendererProperties]}
  {emitterName=trails, rendererClasses [&NiagaraRibbonRendererProperties]}
  {emitterName=SmokeLWE001, rendererClasses []}
  {emitterName=ketchup_mustard_and_food, rendererClasses [&NiagaraMeshRendererProperties]}
 ]
 userVariables []
}
