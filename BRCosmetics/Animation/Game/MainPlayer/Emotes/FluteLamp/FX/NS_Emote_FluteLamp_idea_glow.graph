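# Niagara system for the FluteLamp emote's "idea" glow, a brief warm-yellow sprite flash that follows the skeletal mesh.
# Only idea_flash is enabled; it appears to merge the two disabled Corona_Halo_and_Eyes emitters.
# Two spawn-group bursts feed two sprite renderers, with MergeHelper switches on VisibilityTag setting per-group color, size and camera offset.
# Curves pop the size and fade alpha in and out, with alpha scaled by time of day.
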NiagaraSystem /BRCosmetics/Animation/Game/MainPlayer/Emotes/FluteLamp/FX/NS_Emote_FluteLamp_idea_glow
dependencies {
 usedDataInterfaces [&NiagaraDataInterfaceArrayFloat, &NiagaraDataInterfaceCurve, &NiagaraDataInterfaceSkeletalMesh]
 usedDynamicInputs []
 usedModules [
  &self:SystemSpawnScript.SystemScriptSource.SystemScriptGraph.NiagaraNodeAssignment_0.SetVariables_#1
  &SystemState
 ]
}
emitters {
 Corona_Halo_and_Eyes {
  bEnabled=false
  emitterUpdateScript [
   &EmitterState(
    Inactive Response=::NewEnumerator1
    Life Cycle Mode=::NewEnumerator1
    Loop Behavior=::NewEnumerator1
    Loop Duration Mode=::NewEnumerator1
    UseLoopDelay=false
   )
   &SpawnBurst_Instantaneous(Spawn Group=1, Use Spawn Probability=false, UseLoopCountLimit=false) {module=SpawnBurst_Instantaneous001}
   &SpawnBurst_Instantaneous(Use Spawn Probability=false, UseLoopCountLimit=false)
  ]
  particleSpawnScript [
   /SetVariables(Particles.VisibilityTag=&SpawnGroup()) {isSetParameters=true, module=SetVariables_#2}
   &FN_SetSpawnGroup(Number of Spawn Groups=2)
   &InitializeParticle(
    Color=&MergeHelper_ColorSwitch(0=<rgba 1 0.9852 0.5405 1>, 1=<rgba 1 0.9852 0.5405 0.2>, 2=?, 3=?, Index=$Particles.VisibilityTag, Num=::NewEnumerator0)
    Sprite Size=&MaskVector2dBySpawnGroup(
     Masked Vector2D=&Vector2DFromFloat(Value=&SelectFloatFromArray(Direct Array Index=$Particles.PartitionParticles.Partition, Float Selection Array=DataInterface() {propertyValues {FloatData [44,41.36]}}))
     Passthrough Vector2D=<0 0>
     Spawn Group Mask=0
    )
    Sprite Size Mode=::NewEnumerator4
    UsePositionOffset=false
    Write Material Random=true
   )
   &DynamicMaterialParameters(
    Index 0 Param 1=&MaskFloatBySpawnGroup(Masked Float=0.7, Passthrough Float=1)
    Index 0 Param 2=&MaskFloatBySpawnGroup(Masked Float=0.2, Passthrough Float=1)
    Index 0 Param 3=&MaskFloatBySpawnGroup(Masked Float=0.5, Passthrough Float=1)
    Index 0 Param 4=&MaskFloatBySpawnGroup(Masked Float=1, Passthrough Float=1)
    Index 1 Param 1=&MaskFloatBySpawnGroup(Masked Float=1, Passthrough Float=5)
    Module.Index0Param0Write=true
    Module.Index0Param1Write=true
    Module.Index0Param2Write=true
    Module.Index0Param3Write=true
    Module.Index1Param0Write=true
    Module.Index1Param1Write=false
    Module.Index1Param2Write=false
    Module.Index1Param3Write=false
    Write Parameter Index 1=true
   )
   &CameraOffset(Camera Offset Amount=5)
  ]
  particleUpdateScript [
   &ParticleState(Kill Particles When Lifetime Has Elapsed=false, Let Infinitely Lived Particles Die When Emitter Deactivates=true)
   &SkeletalMeshLocation(Bone Sampling Mode=::NewEnumerator4, MaskSpawnGroup=false, Sampled Position Offset=<10 0 0>, Skeletal Mesh=$System.Skeletal Mesh)
   &ScaleSpriteSize(
    Scale Sprite Size Mode=::NewEnumerator0
    Uniform Scale Factor=&FloatFromCurve(
     FloatCurve=DataInterface() {propertyValues {Curve {
        defaultValue=0
        keys [
         {interpMode=RCIM_Cubic, tangentMode=RCTM_Break, time=0, value=0}
         {interpMode=RCIM_Cubic, leaveTangent=-0.1735, tangentMode=RCTM_Break, time=0.15, value=2.5}
         {arriveTangent=-0.177, interpMode=RCIM_Cubic, leaveTangent=-0.177, tangentMode=RCTM_User, time=0.25, value=1}
         {arriveTangent=-0.15, interpMode=RCIM_Cubic, leaveTangent=-0.15, tangentMode=RCTM_User, time=0.9574, value=1}
        ]
       }}}
     Scale Curve=2.5
    )
   )
   &ScaleColor(
    Scale Alpha=&FloatFromCurve(
     FloatCurve=DataInterface() {propertyValues {Curve {keys [<at 0.0018 0>,<at 0.2561 1>,<at 0.5 0>]}}}
     Scale Curve=&FNTimeOfDayScalar(Night Scale=0.25, Morning Scale=0.3, Day Scale=0.3, Evening Scale=0.3)
    )
    ScaleA=true
    ScaleRGB=true
   )
  ]
  rendererClasses [&NiagaraSpriteRendererProperties]
  renderers [
   {rendererClass=&NiagaraSpriteRendererProperties}
   {rendererClass=&NiagaraSpriteRendererProperties, rendererIndex=1}
  ]
 }
 Corona_Halo_and_Eyes002 {
  bEnabled=false
  emitterUpdateScript [
   &EmitterState(
    Inactive Response=::NewEnumerator1
    Life Cycle Mode=::NewEnumerator1
    Loop Behavior=::NewEnumerator1
    Loop Duration Mode=::NewEnumerator1
    UseLoopDelay=false
   )
   &SpawnBurst_Instantaneous(Spawn Group=1, Use Spawn Probability=false, UseLoopCountLimit=false)
  ]
  particleSpawnScript [
   &InitializeParticle(
    Color=<rgba 1 0.9852 0.5405 0.2>
    Sprite Size=&MaskVector2dBySpawnGroup(
     Masked Vector2D=&Vector2DFromFloat(Value=&SelectFloatFromArray(Direct Array Index=$Particles.PartitionParticles.Partition, Float Selection Array=DataInterface() {propertyValues {FloatData [44,0]}}))
     Passthrough Vector2D=<12 8>
     Spawn Group Mask=0
    )
    Sprite Size Mode=::NewEnumerator4
    UsePositionOffset=false
    Write Material Random=true
   )
   &DynamicMaterialParameters(
    Index 0 Param 1=&MaskFloatBySpawnGroup(Masked Float=0.7, Passthrough Float=0.7)
    Index 0 Param 2=&MaskFloatBySpawnGroup(Masked Float=0.2, Passthrough Float=1)
    Index 0 Param 3=&MaskFloatBySpawnGroup(Masked Float=0.5, Passthrough Float=1)
    Index 0 Param 4=&MaskFloatBySpawnGroup(Masked Float=1, Passthrough Float=1)
    Index 1 Param 1=&MaskFloatBySpawnGroup(Masked Float=1, Passthrough Float=5)
    Module.Index0Param0Write=true
    Module.Index0Param1Write=true
    Module.Index0Param2Write=true
    Module.Index0Param3Write=true
    Module.Index1Param0Write=true
    Module.Index1Param1Write=false
    Module.Index1Param2Write=false
    Module.Index1Param3Write=false
    Write Parameter Index 1=true
   )
   &CameraOffset(Camera Offset Amount=10)
  ]
  particleUpdateScript [
   &ParticleState(Kill Particles When Lifetime Has Elapsed=false, Let Infinitely Lived Particles Die When Emitter Deactivates=true)
   &SkeletalMeshLocation(Bone Sampling Mode=::NewEnumerator4, MaskSpawnGroup=false, Sampled Position Offset=<10 0 0>, Skeletal Mesh=$System.Skeletal Mesh)
   &ScaleSpriteSize(
    Scale Sprite Size Mode=::NewEnumerator0
    Uniform Scale Factor=&FloatFromCurve(FloatCurve=DataInterface() {propertyValues {Curve {
        defaultValue=0
        keys [
         {interpMode=RCIM_Cubic, tangentMode=RCTM_Break, time=0, value=0}
         {interpMode=RCIM_Cubic, leaveTangent=-0.1735, tangentMode=RCTM_Break, time=0.15, value=2.5}
         {arriveTangent=-0.177, interpMode=RCIM_Cubic, leaveTangent=-0.177, tangentMode=RCTM_User, time=0.25, value=1}
         {arriveTangent=-0.15, interpMode=RCIM_Cubic, leaveTangent=-0.15, tangentMode=RCTM_User, time=0.9574, value=1}
        ]
       }}})
   )
   &ScaleColor(
    Scale Alpha=&FloatFromCurve(
     FloatCurve=DataInterface() {propertyValues {Curve {keys [<at 0.0018 0>,<at 0.2561 1>,<at 0.5 0>]}}}
     Scale Curve=&FNTimeOfDayScalar(Night Scale=0.25, Morning Scale=0.3, Day Scale=0.3, Evening Scale=0.3)
    )
    ScaleA=true
    ScaleRGB=true
   )
  ]
  rendererClasses [&NiagaraSpriteRendererProperties]
  renderers [{rendererClass=&NiagaraSpriteRendererProperties}]
 }
 idea_flash {
  emitterUpdateScript [
   &EmitterState()
   &SpawnBurst_Instantaneous(Spawn Group=1, Use Spawn Probability=false, UseLoopCountLimit=false) {module=SpawnBurst_Instantaneous001}
   &SpawnBurst_Instantaneous(Use Spawn Probability=false, UseLoopCountLimit=false)
  ]
  particleSpawnScript [
   /SetVariables(Particles.VisibilityTag=&SpawnGroup()) {isSetParameters=true, module=SetVariables_#2}
   &FN_SetSpawnGroup(Number of Spawn Groups=2)
   &InitializeParticle(
    Color=&MergeHelper_ColorSwitch(0=<rgba 1 0.9852 0.5405 1>, 1=<rgba 1 0.9852 0.5405 0.2>, 2=?, 3=?, Index=$Particles.VisibilityTag, Num=::NewEnumerator0)
    Sprite Size=&MergeHelper_Vector2Switch(0=<44 44>, 1=<12 8>, 2=?, 3=?, 4=?, Index=$Particles.VisibilityTag, Num=::NewEnumerator0)
    Sprite Size Mode=::NewEnumerator4
    UsePositionOffset=false
    Write Material Random=true
   )
   &DynamicMaterialParameters(
    Index 0 Param 1=&MaskFloatBySpawnGroup(Masked Float=0.7, Passthrough Float=1)
    Index 0 Param 2=&MaskFloatBySpawnGroup(Masked Float=0.2, Passthrough Float=1)
    Index 0 Param 3=&MaskFloatBySpawnGroup(Masked Float=0.5, Passthrough Float=1)
    Index 0 Param 4=&MaskFloatBySpawnGroup(Masked Float=1, Passthrough Float=1)
    Index 1 Param 1=&MaskFloatBySpawnGroup(Masked Float=1, Passthrough Float=5)
    Module.Index0Float0Write=true
    Module.Index0Float1Write=true
    Module.Index0Float2Write=true
    Module.Index0Float3Write=true
    Module.Index1Float0Write=true
    Module.Index1Float1Write=false
    Module.Index1Float2Write=false
    Module.Index1Float3Write=false
    Write Parameter Index 1=::NewEnumerator1
   )
   &CameraOffset(Camera Offset Amount=&MergeHelper_FloatSwitch(0=5, 1=10, 2=?, 3=?, Index=$Particles.VisibilityTag, Num=::NewEnumerator0))
  ]
  particleUpdateScript [
   &ParticleState(Kill Particles When Lifetime Has Elapsed=false, Let Infinitely Lived Particles Die When Emitter Deactivates=true)
   &SkeletalMeshLocation(Bone Sampling Mode=::NewEnumerator4, MaskSpawnGroup=false, Sampled Position Offset=<10 0 0>, Skeletal Mesh=$System.Skeletal Mesh)
   &ScaleSpriteSize(
    Scale Sprite Size Mode=::NewEnumerator0
    Uniform Scale Factor=&FloatFromCurve(
     FloatCurve=DataInterface() {propertyValues {Curve {
        defaultValue=0
        keys [
         {interpMode=RCIM_Cubic, tangentMode=RCTM_Break, time=0, value=0}
         {interpMode=RCIM_Cubic, leaveTangent=-0.1735, tangentMode=RCTM_Break, time=0.15, value=2.5}
         {arriveTangent=-0.177, interpMode=RCIM_Cubic, leaveTangent=-0.177, tangentMode=RCTM_User, time=0.25, value=1}
         {arriveTangent=-0.15, interpMode=RCIM_Cubic, leaveTangent=-0.15, tangentMode=RCTM_User, time=0.9574, value=1}
        ]
       }}}
     Scale Curve=&MergeHelper_FloatSwitch(0=2.5, 1=1, 2=?, 3=?, Index=$Particles.VisibilityTag, Num=::NewEnumerator0)
    )
   )
   &ScaleColor(
    Scale Alpha=&FloatFromCurve(
     FloatCurve=DataInterface() {propertyValues {Curve {keys [<at 0.0018 0>,<at 0.2561 1>,<at 0.5 0>]}}}
     Scale Curve=&FNTimeOfDayScalar(Night Scale=0.25, Morning Scale=0.3, Day Scale=0.3, Evening Scale=0.3)
    )
    ScaleA=true
    ScaleRGB=true
   )
  ]
  rendererClasses [&NiagaraSpriteRendererProperties]
  renderers [
   {rendererClass=&NiagaraSpriteRendererProperties}
   {rendererClass=&NiagaraSpriteRendererProperties, rendererIndex=1}
  ]
 }
}
summary {
 systemName=NS_Emote_FluteLamp_idea_glow
 emitters [
  {bEnabled=false, emitterName=Corona_Halo_and_Eyes, rendererClasses [&NiagaraSpriteRendererProperties]}
  {bEnabled=false, emitterName=Corona_Halo_and_Eyes002, rendererClasses [&NiagaraSpriteRendererProperties]}
  {emitterName=idea_flash, rendererClasses [&NiagaraSpriteRendererProperties]}
 ]
 userVariables []
}
