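# Niagara effect for the FluteLamp player emote that draws dark, swirling ribbon trails.
# The chaotic_particles emitter spawns short-lived small sprites driven around a sphere shell by randomized shape-location curves and curl noise.
# The trails emitter reads those particles to spawn near-black ribbons, scales width by sampled sprite size and partly inherits source movement through four ribbon renderers.
# Two socket-sampled ribbon emitters, Dark_ribbon and Multicolor_ribbon, are present but disabled.
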NiagaraSystem /BRCosmetics/Animation/Game/MainPlayer/Emotes/FluteLamp/FX/NS_Emote_FluteLamp_trail
dependencies {
 usedDataInterfaces [
  &NiagaraDataInterfaceCurve
  &NiagaraDataInterfaceColorCurve
  &NiagaraDataInterfaceArrayFloat3
  &NiagaraDataInterfaceArrayFloat
  &NiagaraDataInterfaceVectorCurve
  &NiagaraDataInterfaceParticleRead
  &NiagaraDataInterfaceSocketReader
 ]
 usedDynamicInputs [&Multiply_Vector2D_ByFloat, &FNTimeOfDayColor]
 usedModules [
  &self:SystemSpawnScript.SystemScriptSource.SystemScriptGraph.NiagaraNodeAssignment_3.SetVariables_#1
  &SystemState
  &self:SystemSpawnScript.SystemScriptSource.SystemScriptGraph.NiagaraNodeAssignment_4.SetVariables_#2
  &self:SystemSpawnScript.SystemScriptSource.SystemScriptGraph.NiagaraNodeAssignment_0.SetVariables_#3
 ]
}
emitters {
 Dark_ribbon {
  bEnabled=false
  emitterUpdateScript [
   &EmitterState()
   &SpawnRate(SpawnRate=300, Use Spawn Probability=false)
  ]
  particleSpawnScript [
   &FN_SetSpawnGroup(Number of Spawn Groups=1)
   &PartitionParticles(Number Of Partitions=10, Write Partition to Ribbon ID=::NewEnumerator0)
   &SocketLocation(
    SamplePosOffset=false
    Socket Index=$Particles.PartitionParticles.Partition
    Socket Reader=$System.Socket Reader
    Socket Sampling Mode=::NewEnumerator3
    VelocitySpeedLimit=false
   )
   &InitializeParticle(
    Color Maximum=<rgba 0 0 0 1>
    Color Minimum=<rgba 0.0156 0.0156 0.0156 1>
    Color Mode=::NewEnumerator2
    Lifetime=&RandomRangeFloat(Minimum=0.05, Maximum=0.1)
    Ribbon Width=&RandomRangeFloat(Minimum=20, Maximum=30)
    Ribbon Width Mode=::NewEnumerator1
    UsePositionOffset=false
    Write Material Random=true
   )
   /SetVariables(Particles.Position=&AddVectorToPosition(Position=$Output.SocketLocation.SampledPosition, Vector=<0 0 0>)) {isSetParameters=true, module=SetVariables_#4}
   /SetVariables(Particles.UVDistance:NiagaraFloat=&Multiply_Float(A=&VectorLength(Vector=$Output.SocketLocation.SampledVelocity), B=0.5)) {isSetParameters=true, module=SetVariables_#5}
   &AddVelocity(Velocity=<0 0 500>) {enabled=false}
  ]
  particleUpdateScript [
   &ParticleState()
   /SetVariables(Particles.RibbonUVDistance:NiagaraFloat=$Particles.UVDistance) {isSetParameters=true, module=SetVariables_#6}
   &ScaleRibbonWidth(Ribbon Width Scale=&FloatFromCurve(
     FloatCurve=DataInterface() {propertyValues {Curve {keys [
         {arriveTangent=2.331, interpMode=RCIM_Cubic, leaveTangent=2.331, tangentMode=RCTM_User, time=0, value=0.2}
         {arriveTangent=0.2911, interpMode=RCIM_Cubic, leaveTangent=0.2911, tangentMode=RCTM_User, time=0.0438, value=0.275}
         {arriveTangent=1.32, interpMode=RCIM_Cubic, leaveTangent=1.32, tangentMode=RCTM_User, time=0.5, value=1.9578}
         {arriveTangent=0.3935, interpMode=RCIM_Cubic, leaveTangent=0.3935, tangentMode=RCTM_User, time=1, value=0}
        ]}}}
     Scale Curve=&RandomRangeFloat(Minimum=1, Maximum=1)
    )) {module=ScaleRibbonWidth001}
   &FN_RibbonMaterial(
    Color=&ColorFromCurve(ColorCurve=DataInterface() {propertyValues {
       AlphaCurve {keys [<at 0 2 RCIM_Cubic>,<at 0.9843 0 RCIM_Cubic>]}
       BlueCurve {keys [<at 0.0258 0.5089>]}
       GreenCurve {keys [<at 0.0258 0.4793>]}
       RedCurve {keys [<at 0.0258 0.4735>]}
      }})
    Erosion Amount:NiagaraFloat=&Lerp_Float(A=0.25, B=1, Alpha=$Particles.NormalizedAge)
    Erossion Scale U=4.5
    Macro Distortion Amount=0.1
    Macro Pan Offset=-0.1216
    Micro Distortion Amount=0.0384
    Opacity Erosion:NiagaraFloat=&FloatFromCurve(FloatCurve=DataInterface() {propertyValues {Curve {keys [<at 0 0.25 RCIM_Cubic>, {arriveTangent=1.1077, tangentMode=RCTM_User, time=1, value=1}]}}})
    Opacity Power:NiagaraFloat=&FloatFromCurve(FloatCurve=DataInterface() {propertyValues {Curve {keys [<at 0 0.55 RCIM_Cubic>, {arriveTangent=1.1077, tangentMode=RCTM_User, time=1, value=1}]}}})
    Scale Alpha:NiagaraFloat=&FloatFromCurve(
     FloatCurve=DataInterface() {propertyValues {Curve {keys [<at 0 0.7 RCIM_Cubic>, {arriveTangent=-0.4965, tangentMode=RCTM_User, time=1, value=0}]}}}
     Scale Curve=&FloatFromCurve(
      FloatCurve=DataInterface() {propertyValues {Curve {
         defaultValue=0
         keys [
          {arriveTangent=9.0725, interpMode=RCIM_Cubic, leaveTangent=9.0725, tangentMode=RCTM_User, time=0, value=0}
          {arriveTangent=-2.5249, interpMode=RCIM_Cubic, leaveTangent=-2.5249, tangentMode=RCTM_User, time=0.1408, value=0.9847}
          {arriveTangent=-0.0405, interpMode=RCIM_Cubic, leaveTangent=-0.0405, tangentMode=RCTM_User, time=0.812, value=-0.05}
         ]
        }}}
      Scale Curve=0.2
     )
    )
    UV Basis:NiagaraFloat=&ReturnNormalizedExecIndex()
   ) {module=FN_RibbonMaterial001}
   &ScaleColor(ScaleA=true, ScaleRGB=true)
   &ScaleColor(
    Color Value To Scale=$Particles.Color
    Scale Alpha=&MakeFloatFromLinearColor(LinearColor=$System.TOD Color, Channel:ENiagaraFloatFromLinearColorOptions=::NewEnumerator3)
    ScaleA=true
    ScaleRGB=false
   ) {enabled=false, module=ScaleColor002}
   &SolveForcesAndVelocity(Clamp Velocity=false, Limit Acceleration=false)
   &ScaleRibbonWidth(Ribbon Width Scale=&MakeFloatFromVector2d(Vector 2D=&Vector2DFromFloat(Value=&SelectFloatFromArray(
       Direct Array Index=&SpawnGroup()
       Float Selection Array=DataInterface() {propertyValues {FloatData [0.5,0.8,0.5,0.3,0.3,0.3,0.3,0.3,0.3,0.8,0.8,0.8,0.8,0.5,0.5,0.3,0.3,0.3,0.3,0.4,0.3,0.4,0.4]}}
      ))))
  ]
  rendererClasses [&NiagaraRibbonRendererProperties]
  renderers [
   {rendererClass=&NiagaraRibbonRendererProperties}
   {rendererClass=&NiagaraRibbonRendererProperties, rendererIndex=1}
   {rendererClass=&NiagaraRibbonRendererProperties, rendererIndex=2}
  ]
 }
 Multicolor_ribbon {
  bEnabled=false
  emitterUpdateScript [
   &EmitterState()
   &SpawnRate(SpawnRate=300, Use Spawn Probability=false)
  ]
  particleSpawnScript [
   &PartitionParticles(Number Of Partitions=17, Write Partition to Ribbon ID=::NewEnumerator0)
   &InitializeParticle(
    Color Mode=::NewEnumerator0
    Lifetime=0.12
    Ribbon Width=&RandomRangeFloat(Minimum=20, Maximum=30)
    Ribbon Width Mode=::NewEnumerator1
    UsePositionOffset=false
    Write Material Random=true
   )
   &SocketLocation(
    SamplePosOffset=false
    Socket Index=$Particles.PartitionParticles.Partition
    Socket Reader=$System.Socket Reader
    Socket Sampling Mode=::NewEnumerator3
    VelocitySpeedLimit=false
   )
   &FN_SetSpawnGroup(Number of Spawn Groups=17) {enabled=false}
   /SetVariables(Particles.Position=&AddVectorToPosition(Position=$Output.SocketLocation.SampledPosition, Vector=<0 0 0>)) {isSetParameters=true, module=SetVariables_#4}
   /SetVariables(Particles.UVDistance:NiagaraFloat=&Multiply_Float(A=&VectorLength(Vector=$Output.SocketLocation.SampledVelocity), B=0.5)) {isSetParameters=true, module=SetVariables_#5}
  ]
  particleUpdateScript [
   &ParticleState()
   /SetVariables(Particles.RibbonUVDistance:NiagaraFloat=$Particles.UVDistance) {isSetParameters=true, module=SetVariables_#6}
   &ScaleRibbonWidth(Ribbon Width Scale=&FloatFromCurve(
     FloatCurve=DataInterface() {propertyValues {Curve {keys [
         {arriveTangent=2.331, interpMode=RCIM_Cubic, leaveTangent=2.331, tangentMode=RCTM_User, time=0, value=0.2}
         {arriveTangent=0.2911, interpMode=RCIM_Cubic, leaveTangent=0.2911, tangentMode=RCTM_User, time=0.0438, value=0.275}
         {arriveTangent=1.32, interpMode=RCIM_Cubic, leaveTangent=1.32, tangentMode=RCTM_User, time=0.5, value=1.9578}
         {arriveTangent=0.3935, interpMode=RCIM_Cubic, leaveTangent=0.3935, tangentMode=RCTM_User, time=1, value=0}
        ]}}}
     Scale Curve=&RandomRangeFloat(Minimum=1, Maximum=1)
    )) {module=ScaleRibbonWidth001}
   &FN_RibbonMaterial(
    Color=&ColorFromCurve(ColorCurve=DataInterface() {propertyValues {
       AlphaCurve {keys [<at 0 2 RCIM_Cubic>,<at 0.9843 0 RCIM_Cubic>]}
       BlueCurve {keys [<at 0.0258 0.5089>]}
       GreenCurve {keys [<at 0.0258 0.4793>]}
       RedCurve {keys [<at 0.0258 0.4735>]}
      }})
    Erosion Amount:NiagaraFloat=&Lerp_Float(A=0.25, B=1, Alpha=$Particles.NormalizedAge)
    Erossion Scale U=4.5
    Macro Distortion Amount=0.1
    Macro Pan Offset=-0.1216
    Micro Distortion Amount=0.0384
    Opacity Erosion:NiagaraFloat=&FloatFromCurve(FloatCurve=DataInterface() {propertyValues {Curve {keys [<at 0 0.25 RCIM_Cubic>, {arriveTangent=1.1077, tangentMode=RCTM_User, time=1, value=1}]}}})
    Opacity Power:NiagaraFloat=&FloatFromCurve(FloatCurve=DataInterface() {propertyValues {Curve {keys [<at 0 0.55 RCIM_Cubic>, {arriveTangent=1.1077, tangentMode=RCTM_User, time=1, value=1}]}}})
    Scale Alpha:NiagaraFloat=&FloatFromCurve(
     FloatCurve=DataInterface() {propertyValues {Curve {keys [<at 0 0.7 RCIM_Cubic>, {arriveTangent=-0.4965, tangentMode=RCTM_User, time=1, value=0}]}}}
     Scale Curve=&FloatFromCurve(
      FloatCurve=DataInterface() {propertyValues {Curve {
         defaultValue=0
         keys [
          {arriveTangent=9.0725, interpMode=RCIM_Cubic, leaveTangent=9.0725, tangentMode=RCTM_User, time=0, value=0}
          {arriveTangent=-2.5249, interpMode=RCIM_Cubic, leaveTangent=-2.5249, tangentMode=RCTM_User, time=0.1408, value=0.9847}
          {arriveTangent=-0.0405, interpMode=RCIM_Cubic, leaveTangent=-0.0405, tangentMode=RCTM_User, time=1, value=0}
         ]
        }}}
      Scale Curve=0.2
     )
    )
    UV Basis:NiagaraFloat=&ReturnNormalizedExecIndex()
   ) {module=FN_RibbonMaterial001}
   &ScaleColor(
    Scale Alpha=0.7
    Scale RGB=&SelectVectorFromArray(
     Direct Array Index=$Particles.PartitionParticles.Partition
     Vector Selection Array=DataInterface() {propertyValues {FloatData [
        <1 0.85 0.032>
        <1 0.85 0.032>
        <0 1 0>
        <1 1 1>
        <1 1 1>
        <1 0 0>
        <1 0 0>
        <1 0 0>
        <1 0 0>
        <1 0 0>
        <1 0 0>
        <0 1 0>
        <0 1 0>
        <0 1 0>
        <0 1 0>
        <0 1 0>
        <0 1 0>
        <0 1 0>
       ]}}
    )
    ScaleA=true
    ScaleRGB=true
   )
   &ScaleColor(
    Color Value To Scale=$Particles.Color
    Scale Alpha=&MakeFloatFromLinearColor(LinearColor=$System.TOD Color, Channel:ENiagaraFloatFromLinearColorOptions=::NewEnumerator3)
    ScaleA=true
    ScaleRGB=false
   ) {module=ScaleColor002}
   &SolveForcesAndVelocity(Clamp Velocity=false, Limit Acceleration=false)
   &ScaleRibbonWidth(Ribbon Width Scale=&MakeFloatFromVector2d(Vector 2D=&Vector2DFromFloat(Value=&SelectFloatFromArray(
       Direct Array Index=&SpawnGroup()
       Float Selection Array=DataInterface() {propertyValues {FloatData [0.5,0.8,0.5,0.3,0.3,0.3,0.3,0.3,0.3,0.8,0.8,0.8,0.8,0.5,0.5,0.3,0.3,0.3,0.3,0.4,0.3,0.4,0.4]}}
      ))))
  ]
  rendererClasses [&NiagaraRibbonRendererProperties]
  renderers [
   {rendererClass=&NiagaraRibbonRendererProperties}
   {rendererClass=&NiagaraRibbonRendererProperties, rendererIndex=1}
  ]
 }
 chaotic_particles {
  emitterUpdateScript [
   &EmitterState()
   &SpawnRate(SpawnRate=50, Use Spawn Probability=false)
  ]
  particleSpawnScript [
   &InitializeParticle(
    Lifetime Max=0.55
    Lifetime Min=0.35
    Lifetime Mode=::NewEnumerator1
    Position Mode=::NewEnumerator0
    Sprite Size Mode=::NewEnumerator3
    Uniform Sprite Size=0.05
    Write Material Random=true
   )
   &AddVelocity(Module.Use Falloff Distance=false, Override Seed=false, Velocity Mode=::NewEnumerator1, Velocity Speed=100) {enabled=false}
  ]
  particleUpdateScript [
   &ParticleState()
   &ScaleSpriteSize(
    Uniform Curve Scale=&Lerp_Float(
     A=1
     B=1.8
     Alpha=&FN_LOD_Distance_Remap(Start Distance=100, End Distance=1000, Minimum Output=0, Maximum Output=1, Clamp Output=true)
    )
    Uniform Curve Sprite Scale=DataInterface() {propertyValues {Curve {defaultValue=0, keys [{tangentMode=RCTM_User, time=0, value=0.4118}, <at 0.2094 1>, <at 0.323 1>, {tangentMode=RCTM_User, time=1, value=0}]}}}
   )
   &CurlNoiseForce(
    Bias Noise=false
    Noise Frequency=156.5192
    Noise Strength=&FloatFromCurve(
     FloatCurve=DataInterface() {propertyValues {Curve {keys [<at 0 1>,<at 1 1>]}}}
     Scale Curve=&RandomRangeFloat(Minimum=6000, Maximum=4000)
    )
    Pan Noise=false
   )
   &ShapeLocation(
    MaskSpawnGroup=false
    Offset Coordinate Space=::World
    Offset Mode=::NewEnumerator0
    Rotation Angle:NiagaraFloat=&RandomRangeFloat(Minimum=0, Maximum=1)
    Rotation Axis=<1 1 1>
    Rotation Coordinate Space=::World
    Rotation Mode=::NewEnumerator1
    Sphere Distribution=::NewEnumerator1
    Sphere Radius=45
    Sphere Radius Position:NiagaraFloat=&RandomRangeFloat(Minimum=0.63, Maximum=1)
    U Position=&MakeCustomFloatFromBool(
     Bool=&RandomBool()
     True Float=&FloatFromCurve(
      FloatCurve=DataInterface() {propertyValues {Curve {
         defaultValue=1
         keys [
          {arriveTangent=4.5812, interpMode=RCIM_Cubic, leaveTangent=4.5812, tangentMode=RCTM_User, time=0, value=0}
          {arriveTangent=0.9143, interpMode=RCIM_Cubic, leaveTangent=0.9143, tangentMode=RCTM_User, time=0.2561, value=0.7055}
          {arriveTangent=0.0224, interpMode=RCIM_Cubic, leaveTangent=0.0224, tangentMode=RCTM_User, time=1.0791, value=1}
         ]
        }}}
      CurveIndex=&Multiply_Float(A=$Particles.NormalizedAge, B=&RandomRangeFloat(Minimum=0.1, Maximum=0.5))
      Scale Curve=&RandomRangeFloat(Minimum=1, Maximum=2)
     )
     False Float=&FloatFromCurve(
      FloatCurve=DataInterface() {propertyValues {Curve {keys [{interpMode=RCIM_Cubic, leaveTangent=-1.7692, tangentMode=RCTM_Break, time=0, value=1}, <at 1 0>]}}}
      CurveIndex=&Multiply_Float(A=$Particles.NormalizedAge, B=&RandomRangeFloat(Minimum=0.1, Maximum=0.5))
      Scale Curve=&RandomRangeFloat(Minimum=1, Maximum=2)
     )
    )
    UseOwnerScale=true
    V Position=&MakeCustomFloatFromBool(
     Bool=&RandomBool()
     True Float=&FloatFromCurve(
      FloatCurve=DataInterface() {propertyValues {Curve {
         defaultValue=1
         keys [
          {arriveTangent=4.5812, interpMode=RCIM_Cubic, leaveTangent=4.5812, tangentMode=RCTM_User, time=0, value=0}
          {arriveTangent=1.6567, interpMode=RCIM_Cubic, leaveTangent=1.6567, tangentMode=RCTM_User, time=0.1811, value=0.6464}
          {arriveTangent=0.0224, interpMode=RCIM_Cubic, leaveTangent=0.0224, tangentMode=RCTM_User, time=1, value=1}
         ]
        }}}
      CurveIndex=&Multiply_Float(A=$Particles.NormalizedAge, B=&RandomRangeFloat(Minimum=0.1, Maximum=0.5))
     )
     False Float=&FloatFromCurve(
      FloatCurve=DataInterface() {propertyValues {Curve {keys [
          {interpMode=RCIM_Cubic, leaveTangent=-3.3122, tangentMode=RCTM_Break, time=0, value=1}
          {arriveTangent=-0.4912, tangentMode=RCTM_User, time=1, value=0}
         ]}}}
      CurveIndex=&Multiply_Float(A=$Particles.NormalizedAge, B=&RandomRangeFloat(Minimum=0.1, Maximum=0.5))
     )
    )
   ) {enabled=false}
   &ShapeLocation(
    MaskSpawnGroup=false
    Offset Coordinate Space=::World
    Offset Mode=::NewEnumerator0
    Rotation Angle:NiagaraFloat=&RandomRangeFloat(Minimum=0, Maximum=1)
    Rotation Axis=<1 1 1>
    Rotation Coordinate Space=::World
    Rotation Mode=::NewEnumerator1
    Sphere Distribution=::NewEnumerator1
    Sphere Radius=50
    Sphere Radius Position:NiagaraFloat=&RandomRangeFloat(Minimum=0.5, Maximum=1)
    U Position=&MakeCustomFloatFromBool(
     Bool=&RandomBool()
     True Float=&FloatFromCurve(
      FloatCurve=DataInterface() {propertyValues {Curve {
         defaultValue=1
         keys [
          {arriveTangent=4.5812, interpMode=RCIM_Cubic, leaveTangent=4.5812, tangentMode=RCTM_User, time=0, value=0}
          {arriveTangent=0.9143, interpMode=RCIM_Cubic, leaveTangent=0.9143, tangentMode=RCTM_User, time=0.2561, value=0.7055}
          {arriveTangent=0.0224, interpMode=RCIM_Cubic, leaveTangent=0.0224, tangentMode=RCTM_User, time=1.0791, value=1}
         ]
        }}}
      CurveIndex=&Multiply_Float(A=$Particles.NormalizedAge, B=&RandomRangeFloat(Minimum=0.1, Maximum=0.5))
      Scale Curve=&RandomRangeFloat(Minimum=1, Maximum=2)
     )
     False Float=&FloatFromCurve(
      FloatCurve=DataInterface() {propertyValues {Curve {keys [{interpMode=RCIM_Cubic, leaveTangent=-1.7692, tangentMode=RCTM_Break, time=0, value=1}, <at 1 0>]}}}
      CurveIndex=&Multiply_Float(A=$Particles.NormalizedAge, B=&RandomRangeFloat(Minimum=0.1, Maximum=0.5))
      Scale Curve=&RandomRangeFloat(Minimum=1, Maximum=2)
     )
    )
    UseOwnerScale=true
    V Position=&MakeCustomFloatFromBool(
     Bool=&RandomBool()
     True Float=&FloatFromCurve(
      FloatCurve=DataInterface() {propertyValues {Curve {
         defaultValue=1
         keys [
          {arriveTangent=4.5812, interpMode=RCIM_Cubic, leaveTangent=4.5812, tangentMode=RCTM_User, time=0, value=0}
          {arriveTangent=1.6567, interpMode=RCIM_Cubic, leaveTangent=1.6567, tangentMode=RCTM_User, time=0.1811, value=0.6464}
          {arriveTangent=0.0224, interpMode=RCIM_Cubic, leaveTangent=0.0224, tangentMode=RCTM_User, time=1, value=1}
         ]
        }}}
      CurveIndex=&Multiply_Float(A=$Particles.NormalizedAge, B=&RandomRangeFloat(Minimum=0.1, Maximum=0.5))
     )
     False Float=&FloatFromCurve(
      FloatCurve=DataInterface() {propertyValues {Curve {keys [
          {interpMode=RCIM_Cubic, leaveTangent=-3.3122, tangentMode=RCTM_Break, time=0, value=1}
          {arriveTangent=-0.4912, tangentMode=RCTM_User, time=1, value=0}
         ]}}}
      CurveIndex=&Multiply_Float(A=$Particles.NormalizedAge, B=&RandomRangeFloat(Minimum=0.1, Maximum=0.5))
     )
    )
   ) {module=ShapeLocation001}
   &ScaleVelocity(Velocity Scale=&VectorFromCurve(
     VectorCurve=DataInterface() {propertyValues {
       XCurve {keys [<at 0 0.2>, {interpMode=RCIM_Cubic, tangentMode=RCTM_Break, time=0.5, value=1}, <at 1 0.2>]}
       YCurve {keys [<at 0 0.2>, {interpMode=RCIM_Cubic, tangentMode=RCTM_Break, time=0.5, value=1}, <at 1 0.2>]}
       ZCurve {keys [<at 0 0.2>, {arriveTangent=0.4726, interpMode=RCIM_Cubic, tangentMode=RCTM_Break, time=0.5, value=1}, <at 1 0.2>]}
      }}
     Scale Curve=&RandomRangeVector(Minimum=<20 20 20>, Maximum=<30 30 30>)
    )) {enabled=false}
   &Drag(Drag=&RandomRangeFloat(Minimum=1, Maximum=2), Use Linear Drag=true, Use Rotational Drag=false) {enabled=false}
   &SolveForcesAndVelocity(Clamp Velocity=false, Limit Acceleration=false)
   /SetVariables(Particles.SpriteAlign:Vector3f=&Subtract_Position(A=$Particles.Position, B=$Particles.Previous.Position)) {enabled=false, isSetParameters=true}
  ]
  rendererClasses [&NiagaraSpriteRendererProperties]
  renderers [{rendererClass=&NiagaraSpriteRendererProperties}]
 }
 trails {
  emitterUpdateScript [
   &EmitterState(Life Cycle Mode=::NewEnumerator1, Loop Duration=0.0057, UseLoopDelay=false)
   &SpawnParticlesFromOtherEmitter(
    Attribute Reader=DataInterface() {propertyValues {EmitterBinding {bindingMode=Other, emitterName=chaotic_particles}}}
    Maximum Age Spawn Threshold=1.4
    Spawn Rate=16.0307
    Use Maximum Age=false
    Use Minimum Age=false
   )
  ]
  particleSpawnScript [
   &InitializeParticle(
    Color Maximum=<rgba 0 0 0 1>
    Color Minimum=<rgba 0.024 0.024 0.024 1>
    Color Mode=::NewEnumerator2
    Lifetime=0.5
    Ribbon Width=&RandomRangeFloat(Minimum=200, Maximum=300)
    Ribbon Width Mode=::NewEnumerator1
    UsePositionOffset=false
    Write Material Random=true
   )
   &SampleParticlesFromOtherEmitter(
    Particle ID Sampling=::NewEnumerator1
    Reject Particle If Sampled Index Is Invalid=false
    Sprite Size Sampling:ENiagara_DirectReadSamplingMode=::NewEnumerator1
   )
  ]
  particleUpdateScript [
   &ParticleState()
   /Niagara/Modules/AttributeReader/UpdateParticlesFromOtherEmitter(
    Attribute Reader=DataInterface() {propertyValues {EmitterBinding {bindingMode=Other, emitterName=chaotic_particles}}}
    Position Sampling:ENiagara_DirectReadSamplingMode=::NewEnumerator3
    Sprite Size Sampling:ENiagara_DirectReadSamplingMode=::NewEnumerator1
   )
   &ScaleRibbonWidth(Ribbon Width Scale=&Multiply_Float(
     A=&FloatFromCurve(FloatCurve=DataInterface() {propertyValues {Curve {
         defaultValue=1
         keys [
          {interpMode=RCIM_Cubic, tangentMode=RCTM_User, time=0, value=0}
          {arriveTangent=0.1693, interpMode=RCIM_Cubic, leaveTangent=0.1693, tangentMode=RCTM_User, time=0.2, value=1}
          {interpMode=RCIM_Cubic, tangentMode=RCTM_User, time=0.5, value=1}
          {tangentMode=RCTM_User, time=1, value=0}
         ]
        }}})
     B=&MakeFloatFromVector2d(Vector 2D=$Output.UpdateParticlesFromOtherEmitter.SampledSpriteSize)
    ))
   &InheritSourceMovement(Applied Position Scale=<0.2 0.2 0.2>, ApplyForce=false, ApplyPosition=true, ApplyPositionFalloff=false)
   &SolveForcesAndVelocity(Clamp Velocity=false, Limit Acceleration=false)
   &ScaleColor(Scale Alpha=15, ScaleA=true, ScaleRGB=true)
  ]
  rendererClasses [&NiagaraRibbonRendererProperties]
  renderers [
   {rendererClass=&NiagaraRibbonRendererProperties}
   {rendererClass=&NiagaraRibbonRendererProperties, rendererIndex=1}
   {rendererClass=&NiagaraRibbonRendererProperties, rendererIndex=2}
   {rendererClass=&NiagaraRibbonRendererProperties, rendererIndex=3}
  ]
 }
}
summary {
 systemName=NS_Emote_FluteLamp_trail
 emitters [
  {bEnabled=false, emitterName=Multicolor_ribbon, rendererClasses [&NiagaraRibbonRendererProperties]}
  {bEnabled=false, emitterName=Dark_ribbon, rendererClasses [&NiagaraRibbonRendererProperties]}
  {emitterName=chaotic_particles, rendererClasses [&NiagaraSpriteRendererProperties]}
  {emitterName=trails, rendererClasses [&NiagaraRibbonRendererProperties]}
 ]
 userVariables [
  User.bisFootHidden {type=NiagaraBool}
  User.bIsFrontEnd {type=NiagaraBool}
  User.bIsFrontEndPreview {type=NiagaraBool}
  User.Debug {type=NiagaraBool}
  User.DebugVelocity {type=Vector3f, description="", defaultValue {struct=Vector3f, value=<0 0 1>}}
  User.EmissiveColor {type=LinearColor, description="", defaultValue {struct=LinearColor, value=<rgba 0.0196 0.2745 0.5255 1>}}
  User.ParaGlideLeanAlpha {type=NiagaraFloat, defaultValue=0, description=""}
  User.Radius {type=NiagaraFloat, defaultValue=10, description=""}
  User.RibbonIDOffset {type=NiagaraInt32}
  User.SkeletalMesh {
   type=NiagaraDataInterfaceSkeletalMesh
   defaultValue {
    struct=NiagaraExt_VariableValue_DataInterface
    value {dataInterface=&self:NiagaraDataInterfaceSkeletalMesh_0}
   }
  }
 ]
}
